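# Player AnimBlueprint for the ScubaBondOak item, filling Fortnite item animation layers with an air-dash animation set and a jump-land additive.
# ItemFinalPoseOverride switches between the incoming pose and a nested Air Dashing state machine (intro, Dash1, Dash2, loop states) using bool flags.
# A DivePitch-driven additive blendspace is layered on top, and Dead Blending smooths the inertialized transitions.
# The other layers have no custom logic, and the blueprint declares no variables itself.
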AnimBlueprint /ScubaBondOakGameplay/Assets/Animation/ScubaBondOak_Player_AnimBP
target_skeleton=&Fortnite_M_Avg_Player_Skeleton
graphs [
 AnimGraph {nodes [Output Pose()]}
 EventGraph {links [], nodes []}
 ItemAimOffset {nodes [
   Input Pose!() -> (Pose, AimOffsetAlpha:double, Yaw:double, Pitch:double) {name=InPose_AimOffset}
   Input Pose!() -> (Pose) {name=InPose_PreAimOffsets}
   Output Pose() {layerGroup=Item Override}
  ]}
 ItemFallAdditive {nodes [
   Input Pose() {name=InPoseItemFallAdditive}
   Output Pose() {layerGroup=Item Override}
  ]}
 ItemFallLandAdditive {nodes [
   Input Pose() {name=InPoseItemFallLandAdditive}
   Output Pose(Result="Sequence Player (additive) 'ScubaBondOak_JumpLand'"() {bLoopAnimation=false, sequence=/ScubaBondOakGameplay/Assets/Animation/Jump/ScubaBondOak_JumpLand}) {layerGroup=Item Override}
  ]}
 ItemFallLevel2Override {nodes [
   Input Pose() {name=InPose_FallLevel2}
   Output Pose() {layerGroup=Item Override}
  ]}
 ItemFinalPoseOverride {nodes [
   n1 = Blendspace Player 'BS_DiveAdditive'!(Y=Get DivePitch() -> (DivePitch:float)) -> (Pose) {blendSpace=/ScubaBondOakGameplay/Assets/Proto/BS_DiveAdditive}
   AnimGraphNode_StateMachine_0 = DashingMachine() {is=title AnimGraphNode_StateMachine, bAllowConduitEntryStates=true}
   n2 = Get DivePitchAlpha() -> (DivePitchAlpha:float)
   Output Pose(Result=Dead Blending!(Source:Pose Link Structure=Apply Additive(Base=@AnimGraphNode_StateMachine_0, Additive=@n1, Alpha=@n2) {alphaBoolBlend {blendOption=Linear}}) -> (Pose:Pose Link Structure) {is=title AnimGraphNode_DeadBlending, defaultBlendDuration=0.15}) {layerGroup=Item Override}
   Save cached pose 'InputPose_ItemFinal'(Pose=Input Pose() {name=InFinalPose}) {cacheName=InputPose_ItemFinal}
  ]}
 DashingMachine {
  entry=Main
  path=ItemFinalPoseOverride.AnimGraphNode_StateMachine_0.DashingMachine
  ids=id
  nodes [
   AnimStateNode_15 = Dashing() {is=title AnimStateNode}
   AnimStateNode_4 = Main() {
    is=title AnimStateNode
    stateEntered {guid=#1}
    stateFullyBlended {guid=#2}
    stateLeft {guid=#3}
   }
   AnimStateTransitionNode_0 = AnimStateTransitionNode()
   AnimStateTransitionNode_2 = AnimStateTransitionNode() {automaticRuleTriggerTime=0, logicType=TLT_Inertialization}
  ]
  links [
   {
    dst=Main
    rule=ItemFinalPoseOverride.AnimGraphNode_StateMachine_0.DashingMachine.AnimStateTransitionNode_2.Transition
    src=Dashing
    via=AnimStateTransitionNode_2
   }
   {
    dst=Dashing
    rule=ItemFinalPoseOverride.AnimGraphNode_StateMachine_0.DashingMachine.AnimStateTransitionNode_0.Transition
    src=Main
    via=AnimStateTransitionNode_0
   }
  ]
 }
 Dashing {
  path=ItemFinalPoseOverride.AnimGraphNode_StateMachine_0.DashingMachine.AnimStateNode_15.Dashing
  nodes [
   AnimGraphNode_StateMachine_0 = Air Dashing() {is=title AnimGraphNode_StateMachine}
   Output Animation Pose(Result=@AnimGraphNode_StateMachine_0)
  ]
 }
 Air Dashing {
  entry=Intro
  path=ItemFinalPoseOverride.AnimGraphNode_StateMachine_0.DashingMachine.AnimStateNode_15.Dashing.AnimGraphNode_StateMachine_0.Air Dashing
  ids=id
  nodes [
   AnimStateNode_11 = Dash1 Loop() {is=title AnimStateNode}
   AnimStateNode_13 = Dash2 Loop() {is=title AnimStateNode}
   AnimStateNode_5 = Intro() {is=title AnimStateNode}
   AnimStateNode_7 = Dash1() {is=title AnimStateNode}
   AnimStateNode_8 = Dash2() {is=title AnimStateNode}
   AnimStateTransitionNode_0 = AnimStateTransitionNode() {automaticRuleTriggerTime=0, crossfadeDuration=0.1, logicType=TLT_Inertialization}
   AnimStateTransitionNode_1 = AnimStateTransitionNode() {automaticRuleTriggerTime=0, crossfadeDuration=0.1, logicType=TLT_Inertialization}
   AnimStateTransitionNode_23 = AnimStateTransitionNode() {automaticRuleTriggerTime=0, logicType=TLT_Inertialization}
   AnimStateTransitionNode_5 = AnimStateTransitionNode() {automaticRuleTriggerTime=0, bAutomaticRuleBasedOnSequencePlayerInState=true, logicType=TLT_Inertialization}
   AnimStateTransitionNode_6 = AnimStateTransitionNode() {automaticRuleTriggerTime=0, bAutomaticRuleBasedOnSequencePlayerInState=true, logicType=TLT_Inertialization}
  ]
  links [
   {
    dst=Dash1 Loop
    rule=ItemFinalPoseOverride.AnimGraphNode_StateMachine_0.DashingMachine.AnimStateNode_15.Dashing.AnimGraphNode_StateMachine_0.Air Dashing.AnimStateTransitionNode_5.Transition
    src=Dash1
    via=AnimStateTransitionNode_5
   }
   {
    dst=Dash2
    rule=ItemFinalPoseOverride.AnimGraphNode_StateMachine_0.DashingMachine.AnimStateNode_15.Dashing.AnimGraphNode_StateMachine_0.Air Dashing.AnimStateTransitionNode_23.AnimationTransitionGraph_1
    src=Dash1
    via=AnimStateTransitionNode_1
   }
   {
    dst=Dash2
    rule=ItemFinalPoseOverride.AnimGraphNode_StateMachine_0.DashingMachine.AnimStateNode_15.Dashing.AnimGraphNode_StateMachine_0.Air Dashing.AnimStateTransitionNode_23.AnimationTransitionGraph_1
    src=Dash1 Loop
    via=AnimStateTransitionNode_23
   }
   {
    dst=Dash2 Loop
    rule=ItemFinalPoseOverride.AnimGraphNode_StateMachine_0.DashingMachine.AnimStateNode_15.Dashing.AnimGraphNode_StateMachine_0.Air Dashing.AnimStateTransitionNode_6.Transition
    src=Dash2
    via=AnimStateTransitionNode_6
   }
   {
    dst=Dash1
    rule=ItemFinalPoseOverride.AnimGraphNode_StateMachine_0.DashingMachine.AnimStateNode_15.Dashing.AnimGraphNode_StateMachine_0.Air Dashing.AnimStateTransitionNode_2.Transition
    src=Intro
    via=AnimStateTransitionNode_0
   }
  ]
 }
 Dash1 Loop {
  path=ItemFinalPoseOverride.AnimGraphNode_StateMachine_0.DashingMachine.AnimStateNode_15.Dashing.AnimGraphNode_StateMachine_0.Air Dashing.AnimStateNode_11.Dash1 Loop
  nodes [Output Animation Pose(Result=Sequence Player 'ScubaBondOak_Dash1_Loop'() {bLoopAnimation=false, sequence=/ScubaBondOakGameplay/Assets/Proto/ScubaBondOak_Dash1_Loop})]
 }
 Dash2 Loop {
  path=ItemFinalPoseOverride.AnimGraphNode_StateMachine_0.DashingMachine.AnimStateNode_15.Dashing.AnimGraphNode_StateMachine_0.Air Dashing.AnimStateNode_13.Dash2 Loop
  nodes [Output Animation Pose(Result=Sequence Player 'ScubaBondOak_Dash2_Loop'() {bLoopAnimation=false, sequence=/ScubaBondOakGameplay/Assets/Proto/ScubaBondOak_Dash2_Loop})]
 }
 Intro {
  path=ItemFinalPoseOverride.AnimGraphNode_StateMachine_0.DashingMachine.AnimStateNode_15.Dashing.AnimGraphNode_StateMachine_0.Air Dashing.AnimStateNode_5.Intro
  nodes [Output Animation Pose(Result=Sequence Player 'ScubaBondOak_AscentToDashIntro'() {sequence=/ScubaBondOakGameplay/Assets/Proto/ScubaBondOak_AscentToDashIntro})]
 }
 Dash1 {
  path=ItemFinalPoseOverride.AnimGraphNode_StateMachine_0.DashingMachine.AnimStateNode_15.Dashing.AnimGraphNode_StateMachine_0.Air Dashing.AnimStateNode_7.Dash1
  nodes [Output Animation Pose(Result=Sequence Player 'ScubaBondOak_Dash1'() {bLoopAnimation=false, sequence=/ScubaBondOakGameplay/Assets/Proto/ScubaBondOak_Dash1})]
 }
 Dash2 {
  path=ItemFinalPoseOverride.AnimGraphNode_StateMachine_0.DashingMachine.AnimStateNode_15.Dashing.AnimGraphNode_StateMachine_0.Air Dashing.AnimStateNode_8.Dash2
  nodes [Output Animation Pose(Result=Sequence Player 'ScubaBondOak_Dash2'() {bLoopAnimation=false, sequence=/ScubaBondOakGameplay/Assets/Proto/ScubaBondOak_Dash2})]
 }
 Transition {
  path=ItemFinalPoseOverride.AnimGraphNode_StateMachine_0.DashingMachine.AnimStateNode_15.Dashing.AnimGraphNode_StateMachine_0.Air Dashing.AnimStateTransitionNode_2.Transition
  nodes [Result(bCanEnterTransition=Get bIntoDash1() -> (bIntoDash1:bool))]
 }
 AnimationTransitionGraph_1 {
  path=ItemFinalPoseOverride.AnimGraphNode_StateMachine_0.DashingMachine.AnimStateNode_15.Dashing.AnimGraphNode_StateMachine_0.Air Dashing.AnimStateTransitionNode_23.AnimationTransitionGraph_1
  nodes [Result(bCanEnterTransition=Get bIntoDash2() -> (bIntoDash2:bool))]
 }
 Transition {
  path=ItemFinalPoseOverride.AnimGraphNode_StateMachine_0.DashingMachine.AnimStateNode_15.Dashing.AnimGraphNode_StateMachine_0.Air Dashing.AnimStateTransitionNode_5.Transition
  nodes [Result()]
 }
 Transition {
  path=ItemFinalPoseOverride.AnimGraphNode_StateMachine_0.DashingMachine.AnimStateNode_15.Dashing.AnimGraphNode_StateMachine_0.Air Dashing.AnimStateTransitionNode_6.Transition
  nodes [Result()]
 }
 Main {
  path=ItemFinalPoseOverride.AnimGraphNode_StateMachine_0.DashingMachine.AnimStateNode_4.Main
  nodes [Output Animation Pose(Result=Use cached pose 'InputPose_ItemFinal'())]
 }
 Transition {
  path=ItemFinalPoseOverride.AnimGraphNode_StateMachine_0.DashingMachine.AnimStateTransitionNode_0.Transition
  nodes [Result(bCanEnterTransition=Get bActivated() -> (bActivated:bool))]
 }
 Transition {
  path=ItemFinalPoseOverride.AnimGraphNode_StateMachine_0.DashingMachine.AnimStateTransitionNode_2.Transition
  nodes [Result(bCanEnterTransition=Get bExitTrans() -> (bExitTrans:bool))]
 }
 ItemFullBodyAimOffsets {nodes [
   Input Pose!() -> (Pose, FullBody_AimOffsetAlpha:double, Aim_Yaw:double, Aim_Pitch:double) {name=InPoseFullBodyAimOffsets}
   Input Pose!() -> (Pose) {name=InPosePreFullBodyAimOffsets}
   Output Pose() {layerGroup=Item Override}
  ]}
 ItemFullBodyOverride {nodes [
   Input Pose() {name=InPoseFullBody}
   Output Pose() {layerGroup=Item Override}
  ]}
 ItemIdleOverride {nodes [
   Input Pose() {name=InPose_Idle}
   Output Pose() {layerGroup=Item Override}
  ]}
 ItemJetPackJumpAdditive {nodes [Output Pose() {layerGroup=Item Override}]}
 ItemJetPackStartAdditive {nodes [Output Pose() {layerGroup=Item Override}]}
 ItemJumpLoopAdditive {nodes [
   Input Pose() {name=InPoseItemJumpLoopAdditive}
   Output Pose() {layerGroup=Item Override}
  ]}
 ItemJumpOverride {nodes [
   Input Pose() {name=InPoseJump}
   Output Pose() {layerGroup=Item Override}
  ]}
 ItemJumpUpAdditive {nodes [
   Input Pose() {name=InPoseItemJumpUpAdditive}
   Output Pose() {layerGroup=Item Override}
  ]}
 ItemPreAimOffsets {nodes [
   Input Pose() {name=InPosePreAimOffsets}
   Output Pose() {layerGroup=Item Override}
  ]}
 ItemPreIK {nodes [
   Input Pose() {name=InPose_PreIK}
   Output Pose() {layerGroup=Item Override}
  ]}
 ItemRightHandIK {nodes [
   Input Pose() {name=InPosePreRightHandIK}
   Input Pose() {name=InPosePostRightHandIK}
   Output Pose() {layerGroup=Item Override}
  ]}
 ItemSkeletalControl {nodes [
   Input Pose()
   Output Pose() {layerGroup=Item Override}
  ]}
 ItemSwimJumpFallLoop {nodes [Output Pose() {layerGroup=Item Override}]}
 ItemSwimJumpSurfaceEnd {nodes [Output Pose() {layerGroup=Item Override}]}
 ItemSwimJumpSurfaceLoop {nodes [Output Pose() {layerGroup=Item Override}]}
 ItemUpperBody {nodes [
   Input Pose!() -> (Pose, InputParam:Fort Anim Input Adjusted Aim Structure) {name=InPoseUpperBody}
   Output Pose() {layerGroup=Item Override}
  ]}
 ItemZipLineStartAdditive {nodes [Output Pose() {layerGroup=Item Override}]}
]
variables []
