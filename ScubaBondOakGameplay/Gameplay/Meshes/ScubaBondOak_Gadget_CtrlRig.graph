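# Control Rig for the ScubaBondOak gadget mesh that keeps a dangling cord bone from crossing the handle.
# On the forward solve, it clamps the left cord's second bone to one side of a Y-axis plane in the handle control's space and writes the result back.
# The right cord and all other bones are left untouched.
# Debug drawing of the clamp plane is left enabled.
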ControlRigBlueprint /ScubaBondOakGameplay/Gameplay/Meshes/ScubaBondOak_Gadget_CtrlRig
graphs [
 RigVMFunctionLibrary {links [], nodes []}
 RigVMModel {nodes [
   RigUnit_BeginExecution = &RigUnit_BeginExecution()
   ClampSpatially = /Script/RigVM.RigVMFunction_MathTransformClampSpatially(
    Axis=Y
    DebugColor=(R=1,G=0,B=0,A=1)
    DebugThickness=1
    Maximum=200
    Minimum=0
    Type=Plane
    bDrawDebug=true
    Space=&RigUnit_GetTransform(Item=(Type=Control,Name="handle_ctrl"), bInitial=False).Transform
    Value=&RigUnit_GetTransform(Item=(Type=Bone,Name="cord_02_dyn_l"), bInitial=False).Transform
   ) {
    is=struct RigVMUnitNode
    template="ClampSpatially::Execute(in Value,in Axis,in Type,in Minimum,in Maximum,in Space,in bDrawDebug,in DebugColor,in DebugThickness,out Result)"
   }
   &RigUnit_SetTransform(Item=(Type=Bone,Name="cord_02_dyn_l"), Weight=1, bInitial=False, Value=@ClampSpatially.Result, ExecutePin=@RigUnit_BeginExecution.ExecutePin)
  ]}
]
hierarchy [
 root {loc=<0 0 0>, rot=<0 0 0 1>}
 offset {parent=root, loc=<0 0 0>, rot=<0 0 -0.7071 -0.7071>}
 handle {parent=offset, loc=<0 0 0>, rot=<0 0 0 1>}
 handle_ctrl {controlType=EULER_TRANSFORM, parent=handle, type=Control, loc=<0 0 0>, rot=<0 0 0 1>}
 cord_01_dyn_l {parent=offset, loc=<-2.2609 17.4611 -36.4941>, rot=<0 -0.9763 -0.2164 0>}
 cord_02_dyn_l {parent=cord_01_dyn_l, loc=<5.9408 0 0>, rot=<0 0 0 1>}
 cord_01_dyn_r {parent=offset, loc=<-2.2609 -17.4611 -36.4941>, rot=<0 -0.9764 0.2162 0>}
 cord_02_dyn_r {parent=cord_01_dyn_r, loc=<7.2802 0 0>, rot=<0 0 0 1>}
]
