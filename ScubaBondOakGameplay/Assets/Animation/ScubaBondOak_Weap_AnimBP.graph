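# Animation Blueprint for the ScubaBondOak gadget weapon mesh, adding procedural secondary motion to its left and right cords.
# The reference pose passes through the default montage slot into root-relative Anim Dynamics on each first cord bone, with angular springs and overridden gravity.
# Spring controllers then bounce the second cord bones, and a gadget Control Rig runs last on the local-space pose.
# The event graph is empty and no variables exist.
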AnimBlueprint /ScubaBondOakGameplay/Assets/Animation/ScubaBondOak_Weap_AnimBP
target_skeleton=/ScubaBondOakGameplay/Gameplay/Meshes/ScubaBondOak_Gadget_Skeleton
graphs [
 AnimGraph {nodes [
   n1 = Local Space Ref Pose()
   n2 = Anim Dynamics - Bone: cord_01_dyn_l(ComponentPose=Local To Component(LocalPose=Slot 'DefaultSlot'(Source=@n1)), Alpha=1, ExternalForce="0,0,0") {
    angularDampingOverride=1
    angularSpringConstant=5000
    bAngularSpring=true
    bOverrideAngularDamping=true
    bUseGravityOverride=true
    gravityScale=0.35
    simulationSpace=RootRelative
    alphaBoolBlend {blendOption=Linear}
    boundBone {boneName=cord_01_dyn_l}
    componentLinearAccScale=<0.4 0.4 0.4>
    componentLinearVelScale=<0.4 0.4 0.4>
    gravityOverride {z=-1500}
    physicsBodyDefinitions [{
      boundBone {boneName=cord_01_dyn_l}
      boxExtents=<12 12 12>
      constraintSetup {angularLimitsMax=<0 30 30>, angularLimitsMin=<0 -30 -30>, angularTarget=<1 0 0>, linearAxesMax=<1 1 1>}
      localJointOffset=<5 0 0>
     }]
   }
   n3 = Anim Dynamics - Bone: cord_01_dyn_r(ComponentPose=@n2, Alpha=1, ExternalForce="0,0,0") {
    angularDampingOverride=1
    angularSpringConstant=5000
    bAngularSpring=true
    bOverrideAngularDamping=true
    bUseGravityOverride=true
    gravityScale=0.35
    simulationSpace=RootRelative
    alphaBoolBlend {blendOption=Linear}
    boundBone {boneName=cord_01_dyn_r}
    componentLinearAccScale=<0.4 0.4 0.4>
    componentLinearVelScale=<0.4 0.4 0.4>
    gravityOverride {z=-1500}
    physicsBodyDefinitions [{
      boundBone {boneName=cord_01_dyn_r}
      boxExtents=<8 8 8>
      constraintSetup {angularLimitsMax=<0 30 30>, angularLimitsMin=<0 -30 -30>, angularTarget=<1 0 0>, linearAxesMax=<1 1 1>}
      localJointOffset=<5 0 0>
     }]
   }
   n4 = Spring controller - Bone: cord_02_dyn_l!(ComponentPose:Component Space Pose Link Structure=@n3, Alpha:float=1) -> (Pose:Component Space Pose Link Structure) {
    is=title AnimGraphNode_SpringBone
    springDamping=10
    springStiffness=100
    alphaBoolBlend {blendOption=Linear}
    alphaScaleBiasClamp {scale=0.1}
    springBone {boneName=cord_02_dyn_l}
   }
   n5 = Spring controller - Bone: cord_02_dyn_r!(ComponentPose:Component Space Pose Link Structure=@n4, Alpha:float=1) -> (Pose:Component Space Pose Link Structure) {
    is=title AnimGraphNode_SpringBone
    springDamping=10
    springStiffness=200
    alphaBoolBlend {blendOption=Linear}
    alphaScaleBiasClamp {scale=0.1}
    springBone {boneName=cord_02_dyn_r}
   }
   n6 = Control Rig(Source=Component To Local(ComponentPose=@n5)) {
    alphaBoolBlend {blendOption=Linear}
    controlRigAssetReference {blueprintRigClass=/ScubaBondOakGameplay/Gameplay/Meshes/ScubaBondOak_Gadget_CtrlRig.ScubaBondOak_Gadget_CtrlRig_C}
   }
   Output Pose(Result=@n6)
  ]}
 EventGraph {links [], nodes []}
]
variables []
