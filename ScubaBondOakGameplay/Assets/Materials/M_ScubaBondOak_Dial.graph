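# Material for the ScubaBondOak gadget's dial that animates a needle-like sweep by rotating vertices around a pivot through World Position Offset.
# A custom primitive data InputTime drives a saturated cooldown progress, which sets the sweep angle.
# Summed sines and cosines add wobble, with speed and amplitude switched by Sign-based lerps.
# RotateVertexAboutAxis outputs a masked offset and adjusted normal, with a vertex-red DialMask declared for isolating the dial.
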Material /ScubaBondOakGameplay/Assets/Materials/M_ScubaBondOak_Dial
graph {
 properties {
  bAutomaticallySetUsageInEditor=false
  bUsedWithMeshDeformer=true
  bUsedWithSkeletalMesh=true
  pixelDepthOffsetMode=PDOM_Legacy
  previewMesh=/ScubaBondOakGameplay/Gameplay/Meshes/ScubaBondOak_Gadget
  usedShadingModels=Default Lit
 }
 nodes [
  n1 = Time()
  n2 = ScalarParameter() {defaultValue=0.02, parameterName=JitterFrequency}
  n3 = Multiply(A=Sine(None=Multiply(A=@n1, B=@n2) {constB=0.5}), B=Constant() {r=234.33})
  n4 = Constant() {r=34.455}
  ConstantBiasScale(None=Frac(None=Add(A=@n3, B=@n4))) {bias=-0.5, scale=2}
  MaterialFunctionCall() {materialFunction=&DebugTimeSine}
  n5 = OneMinus(None=NamedRerouteUsage())
  n6 = Multiply(
   A=Time()
   B=LinearInterpolate(A=ScalarParameter() {defaultValue=8, parameterName=FMax}, B=ScalarParameter() {defaultValue=5, parameterName=F}, Alpha=Sign(None=@n5))
  )
  n7 = ScalarParameter() {defaultValue=0.43, parameterName=F1}
  n8 = Multiply(A=Sine(None=Multiply(A=@n6, B=@n7)), B=ScalarParameter() {defaultValue=0.75, parameterName=A1})
  n9 = ScalarParameter() {defaultValue=0.94, parameterName=F2}
  n10 = Multiply(A=Cosine(None=Multiply(A=@n6, B=@n9)), B=ScalarParameter() {defaultValue=1.22, parameterName=A2})
  n11 = Sine(None=Multiply(A=@n6, B=ScalarParameter() {defaultValue=0.67, parameterName=F3}))
  n12 = ScalarParameter() {defaultValue=0.26, parameterName=A3}
  n13 = Multiply(A=@n6, B=ScalarParameter() {defaultValue=0.11, parameterName=F4})
  n14 = Add(
   A=Add(A=Add(A=@n8, B=@n10), B=Multiply(A=@n11, B=@n12))
   B=Multiply(A=Cosine(None=@n13), B=ScalarParameter() {defaultValue=0.11, parameterName=F4})
  )
  n15 = NamedRerouteUsage()
  n16 = LinearInterpolate(
   A=ScalarParameter() {defaultValue=0.03, parameterName=OffsetScaleax}
   B=ScalarParameter() {defaultValue=0.015, parameterName=OffsetScale}
   Alpha=Sign(None=OneMinus(None=@n15))
  )
  n17 = Time()
  n18 = ScalarParameter() {bUseCustomPrimitiveData=true, parameterName=InputTime}
  n19 = Saturate(None=Divide(A=Subtract(A=@n17, B=@n18), B=ScalarParameter() {defaultValue=25, parameterName=CoolDownTime}))
  n20 = Sign(None=@n18)
  n21 = Multiply(A=NamedRerouteDeclaration(None=LinearInterpolate(B=@n19, Alpha=@n20) {constA=1}) {name=AnimProgress0-1, nodeColor=<rgba 1 0 0.6328 1>}) {constB=0.4}
  n22 = MaterialFunctionCall!(
   RotationAngle=Add(A=Multiply(A=@n14, B=@n16), B=Add(A=@n21) {constB=-0.21})
   NormalizedRotationAxis=Transform(None=VectorParameter() {parameterName=RotatingAxis, defaultValue=<rgba 0 1 0 1>}.RGB) {transformSourceType=TRANSFORMSOURCE_Local}
   PivotPoint=TransformPosition(None=VectorParameter() {parameterName=PivotPoint, defaultValue=<rgba 9.5 0 -27.11 1>}.RGB) {transformType=TRANSFORMPOSSOURCE_World}
   Position=TransformPosition(None=LocalPosition().XYZ) {transformType=TRANSFORMPOSSOURCE_World}
   Normal=Transform(None=TextureSampleParameter2D() {parameterName=Normals, samplerType=SAMPLERTYPE_Normal, texture=/Game/Environments/Props/Clocks/Textures/Prop_BigClock_01_N}.RGB)
   TangentU
   TangentV
  ) {materialFunction=/Game/Characters/Player/Male/Medium/Bodies/M_MED_Inferno/FX/Mat/RotateVertexAboutAxis}
  Multiply(A=@n22.WPO, B=NamedRerouteUsage())
  NamedRerouteDeclaration(None=VertexColor().R) {name=DialMask, nodeColor=<rgba 0 0.2266 1 1>}
  SphereMask(A=VectorParameter() {defaultValue=<rgba 0 0 0 1>}.RGB, B=LocalPosition().XYZ) {attenuationRadius=4}
  TransformPosition(None=VectorParameter() {parameterName=PivotPoint, defaultValue=<rgba 9.5 0 -27.11 1>}.RGB) {transformType=TRANSFORMPOSSOURCE_World}
  Transform(None=@n22.Normal) {transformSourceType=TRANSFORMSOURCE_World, transformType=TRANSFORM_Tangent}
 ]
}
