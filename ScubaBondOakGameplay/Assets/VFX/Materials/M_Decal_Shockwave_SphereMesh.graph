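# Translucent unlit Niagara mesh-particle material that, by its name, draws a decal-like shockwave on a sphere mesh, with depth test disabled.
# World position behind translucency builds decal coordinates and box-edge fade.
# Noise-warped sphere masks subtract into a ring, shaped by dynamic Erode and faded near camera, feeding opacity and refraction math.
# Emissive combines particle-colored noise with panning-normal, chromatically split scene color, zeroed on forward and mobile.
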Material /ScubaBondOakGameplay/Assets/VFX/Materials/M_Decal_Shockwave_SphereMesh
graph {
 properties {
  bDisableDepthTest=true
  bUsedWithNiagaraMeshParticles=true
  blendMode=BLEND_Translucent
  pixelDepthOffsetMode=PDOM_Legacy
  refractionMethod=RM_IndexOfRefraction
  shadingModel=MSM_Unlit
  twoSided=true
  usedShadingModels=Unlit
 }
 nodes [
  n1 = Constant() {r=500}
  n2 = NamedRerouteUsage()
  n3 = DynamicParameter() {paramNames [Erode,Ref,Param3,Param4]}
  n4 = NamedRerouteUsage()
  n5 = Add(A=Divide(A=ComponentMask(None=@n4) {g=true, r=true}, B=Constant() {r=100})) {constB=0.5}
  n6 = Multiply(A=Constant() {r=0.8}, B=NamedRerouteUsage())
  n7 = Add(
   A=LinearInterpolate(A=TextureSample(UVs=@n2) {texture=/Game/Effects/Fort_Effects/BluGlow/ZoomNoise}.B, Alpha=@n3.Param3) {constA=0.5, constAlpha=1, constB=0}
   B=NamedRerouteDeclaration(None=MaterialFunctionCall(UVs=@n5, Texture Scale=@n6) {materialFunction=&ScaleUVsByCenter}.UVs) {name=PD_DecalUV, nodeColor=<rgba 1 0.1406 0 1>}
  )
  n8 = Constant() {r=0.5}
  n9 = LinearInterpolate(Alpha=@n3.Erode) {constB=0.5}
  n10 = ScalarParameter() {defaultValue=0.6, parameterName=Hardness}
  n11 = Subtract(
   A=SphereMask(A=@n7, B=@n8, Hardness=Constant() {r=0.8}) {attenuationRadius=0.5}
   B=Multiply(A=SphereMask(A=@n7, B=@n8, Radius=@n9, Hardness=@n10) {attenuationRadius=0.2, hardnessPercent=60}, B=Constant() {r=2})
  )
  n12 = ScalarParameter() {defaultValue=2, parameterName=Brightness}
  n13 = Multiply(
   A=Multiply(
    A=MaterialFunctionCall(Distance=@n1) {materialFunction=&Near_Camera_Fade}.Result
    B=Multiply(A=@n11, B=@n12)
   )
   B=NamedRerouteUsage()
  )
  LinearInterpolate(Alpha=Multiply(A=DynamicParameter() {paramNames [Erode,Ref,Param3,Param4]}.Ref, B=@n13)) {constA=1, constB=1.1}
  n14 = ScreenPosition()
  n15 = Constant() {r=1}
  n16 = Constant() {r=-1}
  n17 = Panner(Coordinate=Multiply(A=TextureCoordinate(), B=AppendVector(A=@n15, B=@n16)), Speed=AppendVector(A=Constant(), B=Constant() {r=1})) {speedY=0.2}
  n18 = Divide(
   A=ComponentMask(None=TextureSample(UVs=@n17) {samplerType=SAMPLERTYPE_Normal, texture=/Game/Global/Textures/Default/Tileable_Patterns/1Normal}.RGB) {g=true, r=true}
   B=ScalarParameter() {defaultValue=20, group=Chromatic, parameterName=ChromaticSceneColorUVDistorion}
  ) {constB=6}
  n19 = SceneColor!(UVs=Add(A=@n14.ViewportUV, B=@n18))
  n20 = SceneColor(UVs=Add(A=@n14.ViewportUV, B=Divide(A=@n18)))
  n21 = SceneColor(UVs=Add(A=@n14.ViewportUV, B=Divide(A=@n18) {constB=3}))
  n22 = Multiply(
   A=MaterialFunctionCall(
    R=MaterialFunctionCall(Float3=@n19.RGB) {materialFunction=&BreakOutFloat3Components}.R
    G=MaterialFunctionCall(Float3=@n20.RGB) {materialFunction=&BreakOutFloat3Components}.G
    B=MaterialFunctionCall(Float3=@n21.RGB) {materialFunction=&BreakOutFloat3Components}.B
   ) {materialFunction=&AppendMany}.RGB
   B=ScalarParameter() {defaultValue=16, group=Chromatic, parameterName=ChromaticIntensity}
  )
  n23 = Constant()
  n24 = TextureSample(UVs=Multiply(A=TextureCoordinate(), B=Constant() {r=2})) {texture=/Game/Effects/Fort_Effects/Textures/Tile/T_Noise_03}
  n25 = Multiply(
   A=Desaturation(None=Power(Base=@n24.RGB), Fraction=Constant() {r=-1})
   B=ScalarParameter() {defaultValue=50, parameterName=BaseNoiseIntensity}
  )
  MaterialFunctionCall!(Blend=Multiply(A=ShadingPathSwitch(Default=@n22, Deferred=@n22, Forward=@n23, Mobile=@n23), B=NamedRerouteUsage()), Base=@n25) {materialFunction=/Engine/Functions/Engine_MaterialFunctions03/Blends/Blend_Lighten}
  n26 = ParticleColor()
  MaterialFunctionCall(Emissive=Multiply(A=@n26.RGB, B=Add(A=@n25))) {materialFunction=&MF_FX_EmissiveAdjustments}
  n27 = ScalarParameter() {defaultValue=1, parameterName=SizeScalar}
  NamedRerouteDeclaration(None=MaterialFunctionCall(X=@n27, Y=@n27, Z=@n27) {materialFunction=&MakeFloat3}.Result) {name=SizeScalar, nodeColor=<rgba 0 0.6744 1 1>}
  n28 = Constant3Vector() {constant=<rgba 0 0 0 1>}
  n29 = NamedRerouteUsage()
  n30 = Divide(A=Add(A=NamedRerouteUsage(), B=Multiply(A=@n28, B=@n29)) {constB=50}, B=NamedRerouteUsage())
  n31 = ScalarParameter() {defaultValue=70, parameterName="DecalEdgeFalloff(Percent)"}
  n32 = Subtract(A=Constant3Vector() {constant=<rgba 100 100 100 1>}, B=Multiply(A=@n31) {constB=2})
  NamedRerouteDeclaration(None=Saturate(None=MaterialFunctionCall!(A=@n30, B, Bounds=@n32, Edge Falloff=@n31) {materialFunction=/Engine/Functions/Engine_MaterialFunctions01/Gradient/BoxMask-3D}.Result)) {name=PD_EdgeOpacity, nodeColor=<rgba 1 0.1406 0 1>}
  n33 = MaterialFunctionCall() {materialFunction=&WorldPositionBehindTranslucency}
  n34 = TransformPosition(None=ParticlePositionWS()) {transformSourceType=TRANSFORMPOSSOURCE_World, transformType=TRANSFORMPOSSOURCE_Instance}
  NamedRerouteDeclaration(None=TransformPosition(None=Subtract(A=@n33.Result, B=@n34)) {transformSourceType=TRANSFORMPOSSOURCE_World, transformType=TRANSFORMPOSSOURCE_Instance}) {name=Position, nodeColor=<rgba 0 0.0859 1 1>}
  Reroute(None=Multiply(A=@n26.A, B=@n13))
  TextureSample() {samplerType=SAMPLERTYPE_Normal, texture=/Game/Characters/Player/Female/Medium/Bodies/F_MED_Bistro_Astronaut/FX/T_GaussianNoise_Normal}
 ]
}
