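# Translucent, unlit, two-sided Niagara mesh-particle material for the ScubaBondOak FlyFlow effect, drawing a panning plasma flow.
# Two panning noise samples sum to distort the plasma and rainbow textures, which form the emissive color, tinted by particle color.
# Inverted Fresnel, plasma, V edge masks and depth fade build opacity and an index-of-refraction blend.
# Dynamic parameter N scales noise distortion of one edge mask, breaking up the flow's edge.
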Material /ScubaBondOakGameplay/Assets/VFX/Materials/M_ScubaBondOak_FlyFlow
graph {
 properties {
  bAutomaticallySetUsageInEditor=false
  bUsedWithNiagaraMeshParticles=true
  blendMode=BLEND_Translucent
  pixelDepthOffsetMode=PDOM_Legacy
  refractionMethod=RM_IndexOfRefraction
  shadingModel=MSM_Unlit
  twoSided=true
  usedShadingModels=Unlit
 }
 nodes [
  Constant3Vector() {constant=<rgba 1 0 0 0>}
  n1 = Fresnel(ExponentIn=ScalarParameter() {defaultValue=2, parameterName="Exponent (S)"}) {baseReflectFraction=0}
  n2 = Saturate(None=Multiply(A=OneMinus(None=@n1), B=ScalarParameter() {defaultValue=1.1, parameterName=ww}))
  n3 = ScalarParameter() {defaultValue=8, parameterName=AW}
  n4 = TextureObjectParameter() {
   group=MainFlowTexture
   parameterName=MainFlowTexture_Texture
   samplerType=SAMPLERTYPE_LinearGrayscale
   sortPriority=0
   texture=/Game/Effects/Fort_Effects/Emotes/UltraLight/Texture/T_Plasma_B_Greyscale
  }
  n5 = TextureObjectParameter() {
   group=NoiseA
   parameterName=NoiseA_Texture
   samplerType=SAMPLERTYPE_LinearGrayscale
   sortPriority=0
   texture=&T_SimplePerlin_Greyscale
  }
  n6 = VectorParameter() {group=NoiseA, parameterName=NoiseA_UVControl, channelNames {a=VPan, b=UPan, g=VScale, r=UScale}, defaultValue=<rgba 1 0.5 0 -1.6002>}
  n7 = VectorParameter() {
   group=NoiseA
   parameterName=NoiseA_NoiseControl
   sortPriority=60
   channelNames {a=VStrength, b=UStrength, g=VOffset, r=UOffset}
   defaultValue=<rgba -0.5 -0.5 0 0>
  }
  n8 = VectorParameter() {group=NoiseB, parameterName=NoiseB_UVControl, channelNames {a=VPan, b=UPan, g=VScale, r=UScale}, defaultValue=<rgba 2 0.5 0 -2.3064>}
  n9 = MaterialFunctionCall(
   Texture=TextureObjectParameter() {group=NoiseB, parameterName=NoiseB_Texture, sortPriority=0, texture=/Game/Items/Traps/Textures/T_PerlinRGB_01_M}
   UVControl_RGB=@n8
   UVControl_A=@n8
  ) {materialFunction=&FXCommon_TextureSample}
  n10 = VectorParameter() {
   group=NoiseB
   parameterName=NoiseB_NoiseControl
   sortPriority=60
   channelNames {a=VStrength, b=UStrength, g=VOffset, r=UOffset}
   defaultValue=<rgba -0.5 -0.2 0 -0.15>
  }
  n11 = Add(
   A=MaterialFunctionCall!(
    NoiseValue=MaterialFunctionCall(Texture=@n5, UVControl_RGB=@n6, UVControl_A=@n6) {materialFunction=&FXCommon_TextureSample}.RG
    Control_RGB=@n7
    Control_A=@n7
   ) {materialFunction=/Game/Effects/Niagara/MasterMaterials/Functions/Common/FXCommon_NoiseControl}.NoiseValue
   B=MaterialFunctionCall(NoiseValue=@n9.RG, Control_RGB=@n10, Control_A=@n10) {materialFunction=&FXCommon_NoiseControl}.NoiseValue
  )
  n12 = VectorParameter() {
   group=MainFlowTexture
   parameterName=MainFlowTexture_UVControl
   channelNames {a=VPan, b=UPan, g=VScale, r=UScale}
   defaultValue=<rgba 1 0.2 0 -2.8074>
  }
  n13 = MaterialFunctionCall(Float2=TextureCoordinate()) {materialFunction=&BreakOutFloat2Components}
  n14 = OneMinus(None=@n13.G)
  n15 = ScalarParameter() {defaultValue=1, parameterName=f2}
  n16 = MaterialFunctionCall(X=@n13.R, Y=OneMinus(None=Power(Base=@n14, Exp=@n15))) {materialFunction=&MakeFloat2}
  n17 = MaterialFunctionCall(Texture=@n4, AdditionalUVOffset=@n11, UVControl_RGB=@n12, UVControl_A=@n12, "[Override]UVInput"=@n16.Result) {materialFunction=&FXCommon_TextureSample}
  n18 = VectorParameter() {parameterName=Color, defaultValue=<rgba 0.03 0.065 0.1 0.5>}
  n19 = Power(Base=Desaturation(None=@n17.RGB, Fraction=Constant() {r=1}), Exp=@n18.A)
  n20 = TextureCoordinate()
  n21 = ScalarParameter() {defaultValue=7, parameterName=A}
  n22 = DynamicParameter() {paramNames [N,Param2,Param3,Param4]}
  n23 = Add(A=@n20, B=Multiply(A=@n11, B=Multiply(A=@n21, B=@n22.N)))
  n24 = VectorParameter() {
   group=EdgeMask
   parameterName=EdgeMask_EdgeMask_InvertControl
   channelNames {a="", b=Power, g=Multiply, r=Offset}
   defaultValue=<rgba -0.1 2 2 1>
  }
  n25 = VectorParameter() {group=EdgeMask, parameterName=EdgeMask_EdgeMask_Control, channelNames {a="", b=Power, g=Multiply, r=Offset}, defaultValue=<rgba 0 5 2 1>}
  n26 = VectorParameter() {
   group=EdgeMask
   parameterName=EdgeMask2_EdgeMask_InvertControl
   channelNames {a="", b=Power, g=Multiply, r=Offset}
   defaultValue=<rgba 5 2 2 1>
  }
  n27 = VectorParameter() {group=EdgeMask, parameterName=EdgeMask2_EdgeMask_Control, channelNames {a="", b=Power, g=Multiply, r=Offset}, defaultValue=<rgba 5 5 2 1>}
  n28 = Multiply(
   A=MaterialFunctionCall!(
    UorV=MaterialFunctionCall(Float2=@n23) {materialFunction=&BreakOutFloat2Components}.G
    Invert_EdgeControl=@n24
    NoInvert_EdgeControl=@n25
   ) {materialFunction=/Game/Effects/Niagara/MasterMaterials/Functions/Common/FXCommon_EdgeMask_Ex}.Mask
   B=MaterialFunctionCall(
    UorV=MaterialFunctionCall(Float2=@n20) {materialFunction=&BreakOutFloat2Components}.G
    Invert_EdgeControl=@n26
    NoInvert_EdgeControl=@n27
   ) {materialFunction=&FXCommon_EdgeMask_Ex}.Mask
  )
  n29 = Multiply(A=Multiply(A=Power(Base=@n2, Exp=@n3), B=Multiply(A=@n19, B=@n28)), B=Power(Base=DepthFade() {fadeDistanceDefault=20}))
  LinearInterpolate(B=ScalarParameter() {defaultValue=0.95, parameterName=Ref}, Alpha=@n29) {constA=1}
  n30 = ParticleColor()
  Multiply(A=@n30.A, B=@n29)
  n31 = TextureObjectParameter() {group=RainbowTexture, parameterName=RainbowTexture_Texture, sortPriority=0, texture=&T_Tile_RainbowColor}
  n32 = VectorParameter() {
   group=RainbowTexture
   parameterName=RainbowTexture_UVControl
   channelNames {a=VPan, b=UPan, g=VScale, r=UScale}
   defaultValue=<rgba 2 0.25 0 -1.7329>
  }
  n33 = MaterialFunctionCall(Texture=@n31, AdditionalUVOffset=@n11, UVControl_RGB=@n32, UVControl_A=@n32, "[Override]UVInput"=@n16.Result) {materialFunction=&FXCommon_TextureSample}
  n34 = Multiply(A=@n33.RGB, B=ScalarParameter() {parameterName=dw})
  n35 = VectorParameter() {defaultValue=<rgba 2 7.5 1 0.5>}
  n36 = Multiply(
   A=Add(A=MaterialFunctionCall(Texture=@n17.RGB, RGB=@n35.RGB, A=@n35.RGB) {materialFunction=&MF_FX_Refine}.Result)
   B=@n18.RGB
  )
  Multiply(
   A=MaterialFunctionCall(Emissive=Add(A=@n34, B=@n36)) {materialFunction=&MF_FX_EmissiveAdjustments}.Result
   B=@n30.RGB
  )
  Multiply(A=Transform(None=@n9.RGB), B=ScalarParameter() {defaultValue=10, parameterName=awf2})
  Multiply(A=Normalize(VectorInput=@n33.RGB), B=@n36)
 ]
}
