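# Additive, unlit, two-sided material for Niagara ribbons, named as the fly trail effect for ScubaBondOak.
# Two panning noise textures distort the UVs of a panning plasma texture sampled with U and V swapped.
# The refined plasma is tinted by a color parameter and dynamic Param1 for emissive.
# A second refinement is multiplied by edge masks along both UV axes and a depth fade to soften the ribbon.
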Material /ScubaBondOakGameplay/Assets/VFX/Materials/M_ScubaBondOak_FlyTrail
graph {
 properties {
  bAutomaticallySetUsageInEditor=false
  bUsedWithNiagaraRibbons=true
  blendMode=BLEND_Additive
  pixelDepthOffsetMode=PDOM_Legacy
  shadingModel=MSM_Unlit
  twoSided=true
  usedShadingModels=Unlit
 }
 nodes [
  n1 = DynamicParameter()
  n2 = VectorParameter() {parameterName=C, defaultValue=<rgba 0 0.1 0.2 1>}
  n3 = TextureObjectParameter() {
   group=MainFlowTexture
   parameterName=MainFlowTexture_Texture
   samplerType=SAMPLERTYPE_LinearGrayscale
   sortPriority=0
   texture=/Game/Effects/Fort_Effects/Emotes/UltraLight/Texture/T_Plasma_B_Greyscale
  }
  n4 = TextureObjectParameter() {
   group=NoiseA
   parameterName=NoiseA_Texture
   samplerType=SAMPLERTYPE_LinearGrayscale
   sortPriority=0
   texture=&T_SimplePerlin_Greyscale
  }
  n5 = VectorParameter() {group=NoiseA, parameterName=NoiseA_UVControl, channelNames {a=VPan, b=UPan, g=VScale, r=UScale}, defaultValue=<rgba 1 0.5 0 -1>}
  n6 = MaterialFunctionCall(Texture=@n4, UVControl_RGB=@n5, UVControl_A=@n5) {materialFunction=&FXCommon_TextureSample}
  n7 = VectorParameter() {
   group=NoiseA
   parameterName=NoiseA_NoiseControl
   sortPriority=60
   channelNames {a=VStrength, b=UStrength, g=VOffset, r=UOffset}
   defaultValue=<rgba -0.5 -0.5 0 0>
  }
  n8 = VectorParameter() {group=NoiseB, parameterName=NoiseB_UVControl, channelNames {a=VPan, b=UPan, g=VScale, r=UScale}, defaultValue=<rgba 2 0.5 0 -1.5>}
  n9 = MaterialFunctionCall(
   Texture=TextureObjectParameter() {group=NoiseB, parameterName=NoiseB_Texture, sortPriority=0, texture=/Game/Items/Traps/Textures/T_PerlinRGB_01_M}
   UVControl_RGB=@n8
   UVControl_A=@n8
  ) {materialFunction=&FXCommon_TextureSample}
  n10 = VectorParameter() {
   group=NoiseB
   parameterName=NoiseB_NoiseControl
   sortPriority=60
   channelNames {a=VStrength, b=UStrength, g=VOffset, r=UOffset}
   defaultValue=<rgba -0.5 -0.5 0 0>
  }
  n11 = VectorParameter() {
   group=MainFlowTexture
   parameterName=MainFlowTexture_UVControl
   channelNames {a=VPan, b=UPan, g=VScale, r=UScale}
   defaultValue=<rgba 0.2 0.05 0 -5>
  }
  n12 = TextureCoordinate()
  n13 = MaterialFunctionCall(
   Texture=@n3
   AdditionalUVOffset=Add(
    A=MaterialFunctionCall!(NoiseValue=@n6.RG, Control_RGB=@n7, Control_A=@n7) {materialFunction=/Game/Effects/Niagara/MasterMaterials/Functions/Common/FXCommon_NoiseControl}.NoiseValue
    B=MaterialFunctionCall(NoiseValue=@n9.RG, Control_RGB=@n10, Control_A=@n10) {materialFunction=&FXCommon_NoiseControl}.NoiseValue
   )
   UVControl_RGB=@n11
   UVControl_A=@n11
   "[Override]UVInput"=MaterialFunctionCall(XY=@n12) {materialFunction=&Swizzle}.YX
  ) {materialFunction=&FXCommon_TextureSample}
  n14 = VectorParameter() {defaultValue=<rgba 1 1 1 0.5>}
  n15 = MaterialFunctionCall(Texture=@n13.RGB, RGB=@n14.RGB, A=@n14.RGB) {materialFunction=&MF_FX_Refine}
  MaterialFunctionCall(Emissive=Multiply(A=Multiply(A=@n1.Param1, B=@n2.RGB), B=Add(A=@n15.Result))) {materialFunction=&MF_FX_EmissiveAdjustments}
  n16 = VectorParameter() {parameterName=Param2, defaultValue=<rgba 0 5 5 0.5>}
  n17 = MaterialFunctionCall(Float2=@n12) {materialFunction=&BreakOutFloat2Components}
  n18 = VectorParameter() {group=EdgeMask, parameterName=EdgeMask_EdgeMask_Control, channelNames {a="", b=Power, g=Multiply, r=Offset}, defaultValue=<rgba 0 2 3 1>}
  n19 = MaterialFunctionCall!(UorV=@n17.G, EdgeControl=@n18) {materialFunction=/Game/Effects/Niagara/MasterMaterials/Functions/Common/FXCommon_EdgeMask_Simple}
  n20 = VectorParameter() {group=TBMsk, parameterName=TBMsk_EdgeMask_InvertControl, channelNames {a="", b=Power, g=Multiply, r=Offset}, defaultValue=<rgba 0 1 3 1>}
  n21 = VectorParameter() {group=TBMsk, parameterName=TBMsk_EdgeMask_Control, channelNames {a="", b=Power, g=Multiply, r=Offset}, defaultValue=<rgba 0 3 2 1>}
  n22 = MaterialFunctionCall!(UorV=@n17.R, Invert_EdgeControl=@n20, NoInvert_EdgeControl=@n21) {materialFunction=/Game/Effects/Niagara/MasterMaterials/Functions/Common/FXCommon_EdgeMask_Ex}
  Multiply(
   A=Multiply(
    A=MaterialFunctionCall(Texture=@n13.RGB, RGB=@n16.RGB, A=@n16.RGB) {materialFunction=&MF_FX_Refine}.Result
    B=Multiply(A=@n19.Mask, B=@n22.Mask)
   )
   B=DepthFade(FadeDistance=ScalarParameter() {defaultValue=30, parameterName=DF})
  )
 ]
}
