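# Additive, unlit, two-sided Niagara ribbon material for a long flight trail with electric plasma flow.
# Two panning lightning samples, scaled by a dynamic brightness parameter, add to a particle-tinted plasma texture distorted by two noise layers and pulsed by a traveling sine wave.
# A center mask boosts the emissive core.
# Opacity combines refined plasma, edge masks along both UV axes, depth fade, near-camera fade and particle alpha.
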Material /ScubaBondOakGameplay/Assets/VFX/Materials/M_ScubaBondOak_FlyTrail_Long
graph {
 properties {
  bAutomaticallySetUsageInEditor=false
  bUsedWithNiagaraRibbons=true
  blendMode=BLEND_Additive
  pixelDepthOffsetMode=PDOM_Legacy
  shadingModel=MSM_Unlit
  twoSided=true
  usedShadingModels=Unlit
 }
 nodes [
  n1 = TextureObjectParameter() {
   group=ThinkLightning
   parameterName=ThinkLightning_Texture
   sortPriority=0
   texture=/Game/Effects/Fort_Effects/Textures/Electric/T_Lightning_Sprites1
  }
  n2 = VectorParameter() {
   group=ThinkLightning
   parameterName=ThinkLightning_UVControl
   channelNames {a=VPan, b=UPan, g=VScale, r=UScale}
   defaultValue=<rgba 0.5 0.3 0 0.25>
  }
  n3 = Constant()
  n4 = TextureCoordinate() {coordinateIndex=1}
  n5 = VectorParameter() {parameterName=CCC, defaultValue=<rgba 6 19.4999 60 7>}
  n6 = Power(
   Base=MaterialFunctionCall(
    Texture=@n1
    AdditionalUVOffset=AppendVector(A=@n2.VPan, B=@n3)
    UVControl_RGB=@n2
    "[Override]UVInput"=MaterialFunctionCall(XY=@n4) {materialFunction=&Swizzle}.YX
   ) {materialFunction=&FXCommon_TextureSample}.R
   Exp=@n5.A
  )
  n7 = TextureObjectParameter() {
   group=ThinkLightning
   parameterName=ThinkLightning_Texture
   sortPriority=0
   texture=&T_Lightning_Sprites1
  }
  n8 = VectorParameter() {
   group=ThinkLightning
   parameterName=ThinkLightning2_UVControl
   channelNames {a=VPan, b=UPan, g=VScale, r=UScale}
   defaultValue=<rgba 0.5 0.3 0 0.75>
  }
  n9 = Constant()
  n10 = TextureCoordinate() {coordinateIndex=1}
  n11 = Power(
   Base=MaterialFunctionCall(
    Texture=@n7
    AdditionalUVOffset=AppendVector(A=@n8.VPan, B=@n9)
    UVControl_RGB=@n8
    "[Override]UVInput"=MaterialFunctionCall(XY=@n10) {materialFunction=&Swizzle}.YX
   ) {materialFunction=&FXCommon_TextureSample}.R
   Exp=@n5.A
  )
  n12 = Multiply(
   A=Multiply(A=Add(A=@n6, B=@n11), B=@n5.RGB)
   B=DynamicParameter() {paramNames [LightningBrightness, Param2, Param3, Param4]}.LightningBrightness
  )
  n13 = Multiply(A=MaterialFunctionCall(Float2=TextureCoordinate()) {materialFunction=&BreakOutFloat2Components}.R) {constB=3}
  n14 = Multiply(A=Time() {bOverride_Period=true, period=10000}, B=ScalarParameter() {defaultValue=2, parameterName=WCW})
  n15 = ConstantBiasScale(None=Sine(None=Add(A=@n13, B=@n14)))
  n16 = ScalarParameter() {defaultValue=2, parameterName=ScaleB}
  n17 = ParticleColor()
  n18 = VectorParameter() {parameterName=C, defaultValue=<rgba 0 0.1 0.2 1>}
  n19 = TextureObjectParameter() {
   group=MainFlowTexture
   parameterName=MainFlowTexture_Texture
   samplerType=SAMPLERTYPE_LinearGrayscale
   sortPriority=0
   texture=/Game/Effects/Fort_Effects/Emotes/UltraLight/Texture/T_Plasma_B_Greyscale
  }
  n20 = TextureObjectParameter() {
   group=NoiseA
   parameterName=NoiseA_Texture
   samplerType=SAMPLERTYPE_LinearGrayscale
   sortPriority=0
   texture=&T_SimplePerlin_Greyscale
  }
  n21 = VectorParameter() {group=NoiseA, parameterName=NoiseA_UVControl, channelNames {a=VPan, b=UPan, g=VScale, r=UScale}, defaultValue=<rgba 1 0.5 0 -1>}
  n22 = MaterialFunctionCall(Texture=@n20, UVControl_RGB=@n21, UVControl_A=@n21) {materialFunction=&FXCommon_TextureSample}
  n23 = VectorParameter() {
   group=NoiseA
   parameterName=NoiseA_NoiseControl
   sortPriority=60
   channelNames {a=VStrength, b=UStrength, g=VOffset, r=UOffset}
   defaultValue=<rgba -0.5 -0.5 0 0>
  }
  n24 = VectorParameter() {group=NoiseB, parameterName=NoiseB_UVControl, channelNames {a=VPan, b=UPan, g=VScale, r=UScale}, defaultValue=<rgba 2 0.5 0 -1.5>}
  n25 = MaterialFunctionCall(
   Texture=TextureObjectParameter() {group=NoiseB, parameterName=NoiseB_Texture, sortPriority=0, texture=/Game/Items/Traps/Textures/T_PerlinRGB_01_M}
   UVControl_RGB=@n24
   UVControl_A=@n24
  ) {materialFunction=&FXCommon_TextureSample}
  n26 = VectorParameter() {
   group=NoiseB
   parameterName=NoiseB_NoiseControl
   sortPriority=60
   channelNames {a=VStrength, b=UStrength, g=VOffset, r=UOffset}
   defaultValue=<rgba -0.5 -0.5 0 0>
  }
  n27 = VectorParameter() {
   group=MainFlowTexture
   parameterName=MainFlowTexture_UVControl
   channelNames {a=VPan, b=UPan, g=VScale, r=UScale}
   defaultValue=<rgba 0.5 0.05 0 -2.5>
  }
  n28 = TextureCoordinate()
  n29 = MaterialFunctionCall(
   Texture=@n19
   AdditionalUVOffset=Add(
    A=MaterialFunctionCall!(NoiseValue=@n22.RG, Control_RGB=@n23, Control_A=@n23) {materialFunction=/Game/Effects/Niagara/MasterMaterials/Functions/Common/FXCommon_NoiseControl}.NoiseValue
    B=MaterialFunctionCall(NoiseValue=@n25.RG, Control_RGB=@n26, Control_A=@n26) {materialFunction=&FXCommon_NoiseControl}.NoiseValue
   )
   UVControl_RGB=@n27
   UVControl_A=@n27
   "[Override]UVInput"=MaterialFunctionCall(XY=@n28) {materialFunction=&Swizzle}.YX
  ) {materialFunction=&FXCommon_TextureSample}
  n30 = VectorParameter() {defaultValue=<rgba 3 50 1 0.5>}
  n31 = Add(A=MaterialFunctionCall(Texture=@n29.RGB, RGB=@n30.RGB, A=@n30.RGB) {materialFunction=&MF_FX_Refine}.Result)
  n32 = Multiply(A=Add(A=Multiply(A=@n15, B=@n16)) {constB=0.2}, B=Multiply(A=@n17.RGB, B=Multiply(A=@n18.RGB, B=@n31)))
  n33 = TextureCoordinate()
  n34 = VectorParameter() {parameterName=TS, defaultValue=<rgba 1 0.3 0 1>}
  n35 = AppendVector(A=@n34.R, B=@n34.R)
  n36 = VectorParameter() {
   group=CenterGlowMask
   parameterName=CenterGlowMask_EdgeMask_InvertControl
   channelNames {a="", b=Power, g=Multiply, r=Offset}
   defaultValue=<rgba 0 2 4 1>
  }
  n37 = VectorParameter() {
   group=CenterGlowMask
   parameterName=CenterGlowMask_EdgeMask_Control
   channelNames {a="", b=Power, g=Multiply, r=Offset}
   defaultValue=<rgba 0 2 4 1>
  }
  n38 = MaterialFunctionCall!(
   UorV=MaterialFunctionCall(Float2=MaterialFunctionCall(UVs=@n33, Texture Scale=@n35) {materialFunction=&ScaleUVsByCenter}.UVs) {materialFunction=&BreakOutFloat2Components}.G
   Invert_EdgeControl=@n36
   NoInvert_EdgeControl=@n37
  ) {materialFunction=/Game/Effects/Niagara/MasterMaterials/Functions/Common/FXCommon_EdgeMask_Ex}
  n39 = ScalarParameter() {defaultValue=10, parameterName=Cebter}
  Multiply(
   A=MaterialFunctionCall(Emissive=Add(A=@n12, B=@n32)) {materialFunction=&MF_FX_EmissiveAdjustments}.Result
   B=Add(A=Multiply(A=@n38.Mask, B=@n39))
  )
  n40 = VectorParameter() {parameterName=Param2, defaultValue=<rgba 0 5 5 0.5>}
  n41 = MaterialFunctionCall(Float2=@n28) {materialFunction=&BreakOutFloat2Components}
  n42 = VectorParameter() {group=EdgeMask, parameterName=EdgeMask_EdgeMask_Control, channelNames {a="", b=Power, g=Multiply, r=Offset}, defaultValue=<rgba 0 2 3 1>}
  n43 = VectorParameter() {group=TBMsk, parameterName=TBMsk_EdgeMask_InvertControl, channelNames {a="", b=Power, g=Multiply, r=Offset}, defaultValue=<rgba 0 1 3 1>}
  n44 = VectorParameter() {group=TBMsk, parameterName=TBMsk_EdgeMask_Control, channelNames {a="", b=Power, g=Multiply, r=Offset}, defaultValue=<rgba 0 3 2 1>}
  n45 = Multiply(
   A=MaterialFunctionCall(Texture=@n29.RGB, RGB=@n40.RGB, A=@n40.RGB) {materialFunction=&MF_FX_Refine}.Result
   B=Multiply(
    A=MaterialFunctionCall!(UorV=@n41.G, EdgeControl=@n42) {materialFunction=/Game/Effects/Niagara/MasterMaterials/Functions/Common/FXCommon_EdgeMask_Simple}.Mask
    B=MaterialFunctionCall(UorV=@n41.R, Invert_EdgeControl=@n43, NoInvert_EdgeControl=@n44) {materialFunction=&FXCommon_EdgeMask_Ex}.Mask
   )
  )
  n46 = DepthFade(FadeDistance=ScalarParameter() {defaultValue=30, parameterName=DF})
  n47 = MaterialFunctionCall(Distance=ScalarParameter() {defaultValue=200, parameterName=D}) {materialFunction=&Near_Camera_Fade}
  Multiply(A=@n17.A, B=Multiply(A=Multiply(A=@n45, B=@n46), B=Power(Base=@n47.Result)))
 ]
}
