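# Masked, lit rock material for ScubaBondOak VFX, usable on Niagara mesh particles, with blue glowing cracks over a desaturated, tinted lava-rock texture.
# A glass-crack mask drives emissive, flickered by layered sines of time whose speed rises with Destabilize.
# Dynamic parameters let particles fade out the glow and scale base color.
# Two rock normal maps are angle-corrected blended, the second intensified through negative flattening.
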Material /ScubaBondOakGameplay/Assets/VFX/Materials/M_ScubaBondOak_Rock
graph {
 properties {bUsedWithNiagaraMeshParticles=true, blendMode=BLEND_Masked, pixelDepthOffsetMode=PDOM_Legacy, usedShadingModels=Default Lit}
 nodes [
  Constant() {r=0.2}
  Constant() {r=0.7}
  Constant() {r=0.5}
  n1 = Multiply(A=TextureCoordinate(), B=ScalarParameter() {defaultValue=1, parameterName=Crack UV})
  n2 = Multiply(
   A=Power(Base=TextureSample(UVs=@n1) {texture=/Game/Effects/Fort_Effects/Textures/Tile/T_Glass_Cracks_M}.G) {constExponent=1.5}
   B=Constant3Vector() {constant=<rgba 0.3125 0.7604 3 0>}
  )
  n3 = ScalarParameter() {defaultValue=0.8, parameterName=FlickerSpeed}
  n4 = Multiply(A=Time(), B=LinearInterpolate(A=@n3, B=Multiply(A=@n3) {constB=2}, Alpha=ScalarParameter() {parameterName=Destabilize}))
  n5 = Add(A=Multiply(B=Sine(None=@n4) {period=0.55}) {constA=0.5}, B=Multiply(A=Sine(None=@n4) {period=0.11}) {constB=0.2})
  n6 = Multiply(A=Sine(None=@n4) {period=0.175}) {constB=0.3}
  n7 = Multiply(A=Sine(None=@n4) {period=3.25}, B=Sine(None=@n4) {period=2.325})
  n8 = LinearInterpolate(Alpha=Multiply(A=Add(A=@n5, B=@n6), B=ConstantBiasScale(None=@n7))) {constA=1, constB=2}
  n9 = DynamicParameter() {defaultValue=<rgba 0 1 1 1>, paramNames [Emissive,BaseColor,Param3,Param4]}
  MaterialFunctionCall(Emissive=LinearInterpolate(A=Multiply(A=@n2, B=@n8), Alpha=@n9.Emissive) {constB=0}) {materialFunction=&MF_FX_EmissiveAdjustments}
  n10 = TextureSampleParameter2D() {parameterName=Normal, samplerType=SAMPLERTYPE_Normal, texture=/Game/Environments/Artemis/Sets/Terrain/Textures/T_Artemis_Lava_Rock_02_N}
  n11 = TextureSample() {samplerType=SAMPLERTYPE_Normal, texture=/Game/Athena/Environments/Volcano/Textures/Rocks_v2_normal}
  n12 = Constant() {r=-1}
  MaterialFunctionCall(
   Normal=MaterialFunctionCall(
    BaseNormal=@n10.RGB
    AdditionalNormal=MaterialFunctionCall(Normal=@n11.RGB, Flatness=@n12) {materialFunction=&FlattenNormal}.Result
   ) {materialFunction=&BlendAngleCorrectedNormals}.Result
   Flatness=ScalarParameter() {parameterName=FlatNormal}
  ) {materialFunction=&FlattenNormal}
  n13 = TextureSampleParameter2D() {parameterName=Base, texture=/Game/Environments/Artemis/Sets/Terrain/Textures/T_Artemis_Lava_Rock_02_D}
  n14 = Multiply(A=ScalarParameter() {defaultValue=1.5, parameterName=BaseColor Multiply}, B=@n9.BaseColor)
  Multiply(
   A=Desaturation(None=Multiply(A=@n13.RGB, B=@n14) {constB=1.5}, Fraction=ScalarParameter() {defaultValue=1, parameterName=Desaturation})
   B=VectorParameter() {parameterName=BaseColor, defaultValue=<rgba 0.75 0.8333 1 0>}.RGB
  )
 ]
}
