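# Translucent, two-sided unlit Niagara material for sprite and mesh particles that draws a simple radial wave for ScubaBondOak VFX.
# Centered UVs become radial coordinates, with one component offset by a scaled flow value and multiplied by a width parameter.
# UseP blends that flow from a time-driven sine to Dynamic Parameter Param1, the default driver.
# The result samples a curve-atlas row to shape the wave, which particle alpha then multiplies.
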Material /ScubaBondOakGameplay/Assets/VFX/Materials/M_ScubaBondOak_SimpleWave
graph {
 properties {
  bUsedWithNiagaraMeshParticles=true
  bUsedWithNiagaraSprites=true
  blendMode=BLEND_Translucent
  pixelDepthOffsetMode=PDOM_Legacy
  shadingModel=MSM_Unlit
  twoSided=true
  usedShadingModels=Unlit
 }
 nodes [
  n1 = TextureCoordinate()
  n2 = MaterialFunctionCall(Float2=MaterialFunctionCall(Vector or UVs=Add(A=@n1) {constB=-0.5}) {materialFunction=&VectorToRadialValue}.Radial Coordinates) {materialFunction=&BreakOutFloat2Components}
  n3 = ConstantBiasScale(None=Sine(None=Time()))
  n4 = DynamicParameter()
  n5 = ScalarParameter() {defaultValue=1, parameterName=UseP}
  n6 = Add(
   A=Multiply(A=LinearInterpolate(A=@n3, B=@n4.Param1, Alpha=@n5), B=ScalarParameter() {defaultValue=-0.47, parameterName=Flow})
   B=ScalarParameter() {defaultValue=0.2, parameterName=FlowOffset}
  )
  n7 = CurveAtlasRowParameter(CurveTime=Multiply(A=Add(A=@n2.G, B=@n6), B=ScalarParameter() {defaultValue=4, parameterName=Widht})) {
   atlas=&/CLA_Temp
   curve=&/CL_Color1
   defaultValue=1
   parameterName=Curve
  }
  Multiply(A=ParticleColor().A, B=@n7.R)
 ]
}
