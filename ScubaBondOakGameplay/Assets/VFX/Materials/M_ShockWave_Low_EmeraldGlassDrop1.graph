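# Additive, unlit, two-sided shockwave material for Niagara mesh particles, built on SplineThicken with a Width parameter.
# A camera-to-normal dot fade, a sine falloff along U and particle alpha combine into a depth-faded mask.
# UseT, zero by default, blends that mask toward a version multiplied by a refined, noise-distorted ET texture.
# A color parameter tints the particle color.
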Material /ScubaBondOakGameplay/Assets/VFX/Materials/M_ShockWave_Low_EmeraldGlassDrop1
graph {
 properties {
  bUsedWithNiagaraMeshParticles=true
  blendMode=BLEND_Additive
  pixelDepthOffsetMode=PDOM_Legacy
  shadingModel=MSM_Unlit
  translucencyPass=MTP_BeforeDOF
  twoSided=true
  usedShadingModels=Unlit
 }
 nodes [
  Constant() {r=1}
  n1 = CameraVectorWS()
  n2 = ScalarParameter() {defaultValue=300, parameterName=Width}
  n3 = MaterialFunctionCall(WidthBase=@n2, WidthTip=@n2, FlattenPixelNormal=StaticBool() {value=true}) {materialFunction=&SplineThicken}
  n4 = Saturate(None=Multiply(A=DotProduct(A=@n1, B=@n3.Normal), B=ScalarParameter() {defaultValue=3, parameterName=FadeSplineThicken_Mul}))
  n5 = ScalarParameter() {defaultValue=3, parameterName=FadeSplineThicken_Power}
  n6 = Multiply(A=MaterialFunctionCall(Float2=TextureCoordinate()) {materialFunction=&BreakOutFloat2Components}.R) {constB=0.5}
  n7 = Multiply(
   A=ScalarParameter() {defaultValue=1, parameterName=awtatatat}
   B=Power(Base=Sine(None=@n6), Exp=ScalarParameter() {defaultValue=1, parameterName=afwafw})
  )
  n8 = ParticleColor()
  n9 = DepthFade(
   Opacity=Multiply(A=Power(Base=@n4, Exp=@n5), B=Multiply(A=@n7, B=@n8.A))
   FadeDistance=ScalarParameter() {defaultValue=200, parameterName=DF}
  ) {fadeDistanceDefault=200}
  n10 = TextureObjectParameter() {group=NoiseA, parameterName=NoiseA_Texture, sortPriority=0}
  n11 = VectorParameter() {group=NoiseA, parameterName=NoiseA_UVControl, channelNames {a=VPan, b=UPan, g=VScale, r=UScale}, defaultValue=<rgba 1 1 0 0>}
  n12 = VectorParameter() {
   group=NoiseA
   parameterName=NoiseA_NoiseControl
   sortPriority=60
   channelNames {a=VStrength, b=UStrength, g=VOffset, r=UOffset}
   defaultValue=<rgba -0.5 -0.5 0.01 0.01>
  }
  n13 = VectorParameter() {group=ET, parameterName=ET_UVControl, channelNames {a=VPan, b=UPan, g=VScale, r=UScale}, defaultValue=<rgba 1 1 0 0>}
  n14 = MaterialFunctionCall(
   Texture=TextureObjectParameter() {group=ET, parameterName=ET_Texture, sortPriority=0}
   AdditionalUVOffset=MaterialFunctionCall!(
    NoiseValue=MaterialFunctionCall(Texture=@n10, UVControl_RGB=@n11, UVControl_A=@n11) {materialFunction=&FXCommon_TextureSample}.RG
    Control_RGB=@n12
    Control_A=@n12
   ) {materialFunction=/Game/Effects/Niagara/MasterMaterials/Functions/Common/FXCommon_NoiseControl}.NoiseValue
   UVControl_RGB=@n13
   UVControl_A=@n13
  ) {materialFunction=&FXCommon_TextureSample}
  n15 = VectorParameter() {parameterName=ETCRefien, defaultValue=<rgba 1 1 1 1>}
  LinearInterpolate(
   A=@n9
   B=Multiply(
    A=MaterialFunctionCall(Texture=@n14.R, RGB=@n15.RGB, A=@n15.RGB) {materialFunction=&MF_FX_Refine}.Result
    B=@n9
   )
   Alpha=ScalarParameter() {parameterName=UseT}
  )
  Multiply(A=VectorParameter() {parameterName=c, defaultValue=<rgba 1 1 1 1>}.RGB, B=@n8.RGB)
  Multiply()
 ]
}
