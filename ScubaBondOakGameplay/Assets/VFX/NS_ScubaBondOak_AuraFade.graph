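# Niagara system for the ScubaBondOak aura fade, a short-lived soft glow on the player, going by its emitter name.
# Its single emitter bursts mesh particles with an overbright white tint and random material value.
# Particle update re-samples System.TargetSM's skeleton each frame so the glow follows it, offset by CameraOffset.
# ScaleColor fades alpha linearly over each particle's life while a waveform modulates RGB brightness; another waveform drives Emitter.OT.
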NiagaraSystem /ScubaBondOakGameplay/Assets/VFX/NS_ScubaBondOak_AuraFade
dependencies {
 usedDataInterfaces [&NiagaraDataInterfaceCurve, &NiagaraDataInterfaceSkeletalMesh]
 usedDynamicInputs []
 usedModules [
  &self:SystemSpawnScript.SystemScriptSource.SystemScriptGraph.NiagaraNodeAssignment_0.SetVariables_#1
  &SystemState
  /Niagara/Noise/Waveform
 ]
}
emitters {SoftGlowOnPlayer {
  emitterUpdateScript [
   &EmitterState(Inactive Response=::NewEnumerator1, Life Cycle Mode=::NewEnumerator1, Loop Behavior=::NewEnumerator1, Loop Duration=3, UseLoopDelay=false)
   &SpawnBurst_Instantaneous(Use Spawn Probability=false, UseLoopCountLimit=false)
   /SetVariables(Emitter.OT:NiagaraFloat=/Niagara/DynamicInputs/Helpers/Waveform("Waveform [1]"=::NewEnumerator12, "[1] Phase Scale"=50, "[1] Frequency"=1, "[1] Amplitude Min/Max"=<2 2.5>)) {isSetParameters=true}
  ]
  particleSpawnScript [&InitializeParticle(
    Color=<rgba 3 3 3 1>
    Lifetime=0.7
    Mesh Scale=<2 1.3 1.3>
    Mesh Scale Mode=::NewEnumerator4
    Sprite Size=<118.0405 198.6727>
    Sprite Size Mode=::NewEnumerator4
    UsePositionOffset=false
    Write Material Random=true
   )]
  particleUpdateScript [
   &ParticleState()
   &SkeletalMeshLocation(Bone Sampling Mode=::NewEnumerator4, MaskSpawnGroup=false, Skeletal Mesh=$System.TargetSM)
   &CameraOffset(Camera Offset Amount=38.6542)
   &ScaleColor(
    Scale Alpha=&FloatFromCurve(FloatCurve=DataInterface() {propertyValues {Curve {keys [<at 0 1>,<at 1 0>]}}})
    Scale RGB=&VectorFromFloat(Value=&Waveform("Waveform [1]"=::NewEnumerator12, "[1] Phase Scale"=50, "[1] Frequency"=1, "[1] Amplitude Min/Max"=<0.5 1>))
    ScaleA=true
    ScaleRGB=true
   )
  ]
  rendererClasses [&NiagaraMeshRendererProperties]
  renderers [{rendererClass=&NiagaraMeshRendererProperties, rendererIndex=0}]
 }}
summary {
 systemName=NS_ScubaBondOak_AuraFade
 emitters [{emitterName=SoftGlowOnPlayer, rendererClasses [&NiagaraMeshRendererProperties]}]
 userVariables []
}
