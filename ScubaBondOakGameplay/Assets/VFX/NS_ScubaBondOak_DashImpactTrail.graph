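# Niagara system for the ScubaBondOak dash impact trail, a long mesh trail on the weapon that spawns only for non-local players.
# Two burst spawn groups follow the WeaponMesh, with randomized width and rotation and a heavy Z stretch.
# An age-gated update aligns the meshes to DashDirection or owner velocity, selected by UseSystemV.
# Fade, material parameters and kill key off real time since JumpTime plus TotalSkillTime, tying removal to skill end.
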NiagaraSystem /ScubaBondOakGameplay/Assets/VFX/NS_ScubaBondOak_DashImpactTrail
dependencies {
 usedDataInterfaces [&NiagaraDataInterfaceCurve]
 usedDynamicInputs []
 usedModules [
  &self:SystemSpawnScript.SystemScriptSource.SystemScriptGraph.NiagaraNodeAssignment_0.SetVariables_#1
  &SystemState
  &self:SystemSpawnScript.SystemScriptSource.SystemScriptGraph.NiagaraNodeAssignment_1.SetVariables_#2
 ]
}
emitters {LongTrail_NLocalOnly_ {
  emitterUpdateScript [
   &EmitterState(Life Cycle Mode=::NewEnumerator1, Loop Behavior=::NewEnumerator1, Loop Duration=2, UseLoopDelay=true)
   &SpawnBurst_Instantaneous(
    Spawn Count=4
    Spawn Probability=&MakeCustomFloatFromBool(Bool=$User.IsLocalPlayer, True Float=0, False Float=1)
    Use Spawn Probability=true
    UseLoopCountLimit=false
   )
   &SpawnBurst_Instantaneous(
    Spawn Count=2
    Spawn Group=1
    Spawn Probability=&MakeCustomFloatFromBool(Bool=$User.IsLocalPlayer, True Float=0, False Float=1)
    Use Spawn Probability=true
    UseLoopCountLimit=false
   ) {module=SpawnBurst_Instantaneous001}
  ]
  particleSpawnScript [
   &InitializeParticle(
    Color=&MaskLinearColorBySpawnGroup(Masked Linear Color=<rgba 1 1 1 0.7>, Passthrough Linear Color=<rgba 2 2 2 0.4>)
    Lifetime=&RandomRangeFloat(Minimum=12, Maximum=12)
    Mesh Scale Mode=::NewEnumerator3
    Mesh Uniform Scale=&Multiply_Float(
     A=&MaskFloatBySpawnGroup(
      Masked Float=&RandomRangeFloat(Minimum=2.2882, Maximum=4.2376)
      Passthrough Float=&RandomRangeFloat(Minimum=3, Maximum=3.5)
     )
     B=0.5
    )
    Position Offset Coordinate Space=::World
    UsePositionOffset=true
    Write Material Random=true
   )
   &SkeletalMeshLocation(
    MaskSpawnGroup=false
    Mesh Sampling Type=::NewEnumerator5
    Orientation Sampling=::NewEnumerator1
    Skeletal Mesh=$User.WeaponMesh
    Socket Sampling Mode=::NewEnumerator5
   )
   &InitialMeshOrientation(
    Mesh Axis To Orient=<0 0 1>
    Mesh Orientation Mode=::NewEnumerator2
    Orientation Vector=<0 0 1>
    Rotation=&MaskVectorBySpawnGroup(
     Masked Vector=&RandomRangeVector(Minimum=<0 0 0>, Maximum=<0 0 360>)
     Passthrough Vector=&RandomRangeVector(Minimum=<0 0 0>, Maximum=<0 0 360>)
    )
    Use Rotation Vector=true
   )
  ]
  particleUpdateScript [
   &ParticleState()
   &SkeletalMeshLocation(
    MaskSpawnGroup=false
    Mesh Sampling Type=::NewEnumerator5
    Orientation Sampling=::NewEnumerator1
    Position Sampling=::NewEnumerator2
    Skeletal Mesh=$User.WeaponMesh
    Socket Sampling Mode=::NewEnumerator5
   ) {module=SkeletalMeshLocation001}
   &ScaleMeshSize(Scale Factor=&MaskVectorBySpawnGroup(
     Masked Vector=&RandomRangeVector(Minimum=<3 3 15>, Maximum=<4 4 15>)
     Passthrough Vector=&RandomRangeVector(Minimum=<1 1 15>, Maximum=<2 2 15>)
    )) {module=ScaleMeshSize001}
   &ScaleMeshSize(Scale Factor=&MakeVector(
     X=1
     Y=1
     Z=&FloatFromCurve(
      FloatCurve=DataInterface() {propertyValues {Curve {keys [
          {interpMode=RCIM_Cubic, leaveTangent=0.0966, tangentMode=RCTM_Break, time=0, value=1}
          {arriveTangent=0.0951, interpMode=RCIM_Cubic, leaveTangent=-0.1745, tangentMode=RCTM_Break, time=1, value=1.07}
         ]}}}
      Scale Curve=1.5
     )
    ))
   &DynamicMaterialParameters(
    Index 0 Param 1=&MakeFloatFromVector(Vector=$Particles.Scale, Channel=::NewEnumerator0)
    Index 0 Param 2=&MaskFloatBySpawnGroup(Masked Float=8, Passthrough Float=9)
    Index 0 Param 3=&Multiply_Float(
     A=&FloatFromCurve(
      FloatCurve=DataInterface() {propertyValues {Curve {keys [{interpMode=RCIM_Cubic, leaveTangent=1.2022, tangentMode=RCTM_Break, time=0, value=1}, {tangentMode=RCTM_User, time=2, value=0}]}}}
      CurveIndex=&Subtract_Float(A=&Subtract_Float(A=$Engine.RealTime, B=$User.JumpTime), B=$User.TotalSkillTime)
      Scale Curve=&MaskFloatBySpawnGroup(Masked Float=0.015, Passthrough Float=0.25)
     )
     B=0.7
    )
    Index 0 Param 4=1.2
    Index 1 Param 1=&FloatFromCurve(
     FloatCurve=DataInterface() {propertyValues {Curve {keys [
         {interpMode=RCIM_Cubic, leaveTangent=1.1034, tangentMode=RCTM_Break, time=0, value=0}
         {arriveTangent=2.2399, tangentMode=RCTM_User, time=0.7, value=1}
        ]}}}
     CurveIndex=$Particles.Age
    )
    Index 1 Param 2=&FloatFromCurve(
     FloatCurve=DataInterface() {propertyValues {Curve {keys [<at 0 0>, {arriveTangent=3.2053, tangentMode=RCTM_User, time=1.5, value=1}]}}}
     CurveIndex=&Subtract_Float(A=&Subtract_Float(A=$Engine.RealTime, B=$User.JumpTime), B=$User.TotalSkillTime)
    )
    Index 1 Param 4=&FloatFromCurve(
     FloatCurve=DataInterface() {propertyValues {Curve {keys [
         {interpMode=RCIM_Cubic, leaveTangent=-0.0025, tangentMode=RCTM_Break, time=0, value=0}
         {arriveTangent=1.8528, interpMode=RCIM_Cubic, tangentMode=RCTM_Break, time=1, value=1}
        ]}}}
     Scale Curve=&RandomRangeFloat(Minimum=40, Maximum=60)
    )
    Index 2 Param 1=&FloatFromCurve(FloatCurve=DataInterface() {propertyValues {Curve {keys [
         {interpMode=RCIM_Cubic, leaveTangent=-0.0025, tangentMode=RCTM_Break, time=0, value=0.5}
         {arriveTangent=-0.0316, interpMode=RCIM_Cubic, tangentMode=RCTM_Break, time=1, value=-0.2}
        ]}}})
    Module.Index0Param0Write=true
    Module.Index0Param1Write=true
    Module.Index0Param2Write=true
    Module.Index0Param3Write=true
    Module.Index1Param0Write=true
    Module.Index1Param1Write=true
    Module.Index1Param2Write=true
    Module.Index1Param3Write=true
    Module.Index2Param0Write=true
    Module.Index2Param1Write=true
    Module.Index2Param2Write=true
    Module.Index2Param3Write=true
    Module.Index3Param0Write=true
    Module.Index3Param1Write=true
    Module.Index3Param2Write=true
    Module.Index3Param3Write=true
    Write Parameter Index 1=true
    Write Parameter Index 2=true
    Write Parameter Index 3=true
   )
   &SolveForcesAndVelocity(Clamp Velocity=false, Limit Acceleration=false)
   &ScaleColor(
    Scale Alpha=&FloatFromCurve(
     FloatCurve=DataInterface() {propertyValues {Curve {keys [
         {tangentMode=RCTM_User, time=0, value=1}
         {arriveTangent=-0.0405, interpMode=RCIM_Cubic, leaveTangent=-0.0405, tangentMode=RCTM_User, time=0.5, value=1}
         {arriveTangent=-0.0045, interpMode=RCIM_Cubic, leaveTangent=-0.0045, tangentMode=RCTM_User, time=2, value=0}
        ]}}}
     CurveIndex=&Subtract_Float(A=&Subtract_Float(A=$Engine.RealTime, B=$User.JumpTime), B=$User.TotalSkillTime)
    )
    Scale RGB=&VectorFromFloat(Value=&RandomRangeFloat(
      Minimum=&MaskFloatBySpawnGroup(Masked Float=0.2, Passthrough Float=0.5)
      Maximum=1
      Evaluation Type=::NewEnumerator1
     ))
    ScaleA=true
    ScaleRGB=true
   )
   /SetVariables(Particles.MeshOrientation=/Niagara/DynamicInputs/TypeConversions/MakeCustomQuaternionFromBool(
     Bool:NiagaraBool=&SetBoolByFloatComparison(
      A=$Particles.Age
      B=&MakeCustomFloatFromBool(Bool=$User.UseSystemV, True Float=1, False Float=0.3)
      Comparison Type=::NewEnumerator6
     )
     True Quat:Quat4f=&MultiplyQuaternion(
      Quaternion A=&FindQuatBetweenTwoVectors(
       From Vector=<0 0 -1>
       To Vector=&MakeCustomVectorFromBool(
        Bool=$User.UseSystemV
        True Vector=&NormalizeVector(
         Vector To Normalize=&Multiply_VectorByFloat(Vector=$Engine.Owner.Velocity, Float=1)
         Safe Normalize=true
         Safe Fallback Threshold=0
         Safe Fallback Vector=<0 0 0.001>
        )
        False Vector=$User.DashDirection
       )
      )
      Quaternion B=$Particles.Initial.MeshOrientation
     )
     False Quat:Quat4f=$Particles.MeshOrientation
    )) {isSetParameters=true}
   &KillParticles(Kill Particles=&SetBoolByFloatComparison(
     A=&Subtract_Float(A=&Subtract_Float(A=$Engine.RealTime, B=$User.JumpTime), B=$User.TotalSkillTime)
     B=2
    ))
  ]
  rendererClasses [&NiagaraMeshRendererProperties]
  renderers [
   {rendererClass=&NiagaraMeshRendererProperties, rendererIndex=0}
   {rendererClass=&NiagaraMeshRendererProperties, rendererIndex=1}
  ]
 }}
summary {
 systemName=NS_ScubaBondOak_DashImpactTrail
 emitters [{emitterName=LongTrail_NLocalOnly_, rendererClasses [&NiagaraMeshRendererProperties]}]
 userVariables [
  User.DashDirection {type=Vector3f, description="", defaultValue {struct=Vector3f, value=<0 0 0>}}
  User.IsLocalPlayer {type=NiagaraBool}
  User.JumpTime {type=NiagaraFloat, defaultValue=99999, description=""}
  User.TotalSkillTime {type=NiagaraFloat, defaultValue=8, description=""}
  User.UseSystemV {type=NiagaraBool}
  User.WeaponMesh {
   type=NiagaraDataInterfaceSkeletalMesh
   defaultValue {
    struct=NiagaraExt_VariableValue_DataInterface
    value {dataInterface=&self:NiagaraDataInterfaceSkeletalMesh_0}
   }
  }
 ]
}
