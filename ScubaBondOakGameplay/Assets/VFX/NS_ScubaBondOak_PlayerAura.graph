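# Niagara player aura for ScubaBondOak, combining ribbon trails with a pulsing soft glow on a target skeletal mesh.
# RibbonOnHandAndFoot spawns three movement-driven groups at separate bones or sockets, offset along owner velocity.
# A spawn-group mask varies one group's width, lifetime, offset and material parameter.
# SoftGlowOnPlayer keeps a burst-spawned mesh particle following the skeleton, shifted by camera offset, with waveform-flickered color.
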NiagaraSystem /ScubaBondOakGameplay/Assets/VFX/NS_ScubaBondOak_PlayerAura
dependencies {
 usedDataInterfaces [&NiagaraDataInterfaceArrayInt32, &NiagaraDataInterfaceSkeletalMesh]
 usedDynamicInputs []
 usedModules [
  &self:SystemSpawnScript.SystemScriptSource.SystemScriptGraph.NiagaraNodeAssignment_0.SetVariables_#1
  &SystemState
  /Niagara/Noise/Waveform
 ]
}
emitters {
 RibbonOnHandAndFoot {
  emitterUpdateScript [
   &EmitterState(Life Cycle Mode=::NewEnumerator1, Loop Behavior=::NewEnumerator1, Loop Duration=4, UseLoopDelay=false)
   &SpawnPerUnit(Module.Use Max Movement Threshold=false, Module.Use Movement Tolerance=false, Spawn Spacing=30, Use Spawn Probability=false)
   &SpawnPerUnit(Module.Use Max Movement Threshold=false, Module.Use Movement Tolerance=false, Spawn Group=1, Spawn Spacing=30, Use Spawn Probability=false) {module=SpawnPerUnit001}
   &SpawnPerUnit(Module.Use Max Movement Threshold=false, Module.Use Movement Tolerance=false, Spawn Group=2, Spawn Spacing=30, Use Spawn Probability=false) {module=SpawnPerUnit002}
  ]
  particleSpawnScript [
   &InitializeParticle(
    Color Mode=::NewEnumerator0
    Lifetime=&MaskFloatBySpawnGroup(Masked Float=0.2, Passthrough Float=0.17, Spawn Group Mask=2)
    Ribbon Width=&MaskFloatBySpawnGroup(Masked Float=100, Passthrough Float=50, Spawn Group Mask=2)
    Ribbon Width Mode=::NewEnumerator1
    UsePositionOffset=false
    Write Material Random=true
   )
   &SkeletalMeshLocation(
    Bone / Socket Index=&SelectIntFromArray(Direct Array Index=&SpawnGroup(), Int Selection Array=DataInterface() {propertyValues {IntData [1,3,2]}})
    Bone Sampling Mode=::NewEnumerator4
    MaskSpawnGroup=false
    Sampled Position Offset=&Multiply_VectorByFloat(
     Vector=&NormalizeVector(Vector To Normalize=$Engine.Owner.Velocity, Safe Normalize=true, Safe Fallback Threshold=0, Safe Fallback Vector=<0 0 0.001>)
     Float=&MaskFloatBySpawnGroup(Masked Float=100, Passthrough Float=50, Spawn Group Mask=2)
    )
    Sampled Position Offset Coordinate Space:ENiagaraSkelMeshTransforms=::NewEnumerator1
    Skeletal Mesh=$System.TargetSM
   )
   &DynamicMaterialParameters(
    Index 0 Param 1=&MaskFloatBySpawnGroup(Masked Float=0.5, Passthrough Float=1, Spawn Group Mask=2)
    Module.Index0Param0Write=true
    Module.Index0Param1Write=true
    Module.Index0Param2Write=true
    Module.Index0Param3Write=true
   )
  ]
  particleUpdateScript [&ParticleState()]
  rendererClasses [&NiagaraRibbonRendererProperties]
  renderers [{rendererClass=&NiagaraRibbonRendererProperties}]
 }
 SoftGlowOnPlayer {
  emitterUpdateScript [
   &EmitterState(
    Inactive Response=::NewEnumerator1
    Life Cycle Mode=::NewEnumerator1
    Loop Behavior=::NewEnumerator1
    Loop Duration Mode=::NewEnumerator1
    UseLoopDelay=false
   )
   &SpawnBurst_Instantaneous(Use Spawn Probability=false, UseLoopCountLimit=false)
   /SetVariables(Emitter.OT:NiagaraFloat=/Niagara/DynamicInputs/Helpers/Waveform("Waveform [1]"=::NewEnumerator12, "[1] Phase Scale"=50, "[1] Frequency"=1, "[1] Amplitude Min/Max"=<2 2.5>)) {isSetParameters=true}
  ]
  particleSpawnScript [&InitializeParticle(
    Color=<rgba 3 3 3 1>
    Lifetime=2
    Mesh Scale=<2 1.3 1.3>
    Mesh Scale Mode=::NewEnumerator4
    Sprite Size=<118.0405 198.6727>
    Sprite Size Mode=::NewEnumerator4
    UsePositionOffset=false
    Write Material Random=true
   )]
  particleUpdateScript [
   &ParticleState(Kill Particles When Lifetime Has Elapsed=false, Let Infinitely Lived Particles Die When Emitter Deactivates=true)
   &SkeletalMeshLocation(Bone Sampling Mode=::NewEnumerator4, MaskSpawnGroup=false, Skeletal Mesh=$System.TargetSM)
   &CameraOffset(Camera Offset Amount=38.6542)
   &ScaleColor(
    Scale RGB=&VectorFromFloat(Value=&Waveform("Waveform [1]"=::NewEnumerator12, "[1] Phase Scale"=50, "[1] Frequency"=1, "[1] Amplitude Min/Max"=<0.5 1>))
    ScaleA=true
    ScaleRGB=true
   )
  ]
  rendererClasses [&NiagaraMeshRendererProperties]
  renderers [{rendererClass=&NiagaraMeshRendererProperties, rendererIndex=0}]
 }
}
summary {
 systemName=NS_ScubaBondOak_PlayerAura
 emitters [
  {emitterName=SoftGlowOnPlayer, rendererClasses [&NiagaraMeshRendererProperties]}
  {emitterName=RibbonOnHandAndFoot, rendererClasses [&NiagaraRibbonRendererProperties]}
 ]
 userVariables []
}
