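# Niagara system for the ScubaBondOak rift: a blue energy effect with layered meshes, looping shockwaves, flares and a point light.
# EnvChecker traces downward and writes collision points into a user array, which LinkLightning reads to grow bent ribbon arcs with glowing tips and location events, alongside a LinkSpark emitter.
# Merged emitters switch looks by spawn group or partition, while user parameters drive enable fade, offset, scale and front-end adjustments.
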NiagaraSystem /ScubaBondOakGameplay/Assets/VFX/NS_ScubaBondOak_Rift
dependencies {
 usedDataInterfaces [
  &NiagaraDataInterfaceCurve
  &NiagaraDataInterfaceCollisionQuery
  &NiagaraDataInterfaceDebugDraw
  &NiagaraDataInterfaceColorCurve
  &NiagaraDataInterfaceVector2DCurve
 ]
 usedDynamicInputs []
 usedModules [&SystemState]
}
emitters {
 EnergyWave {
  emitterUpdateScript [
   &EmitterState(Life Cycle Mode=::NewEnumerator1, Loop Behavior=::NewEnumerator1, Loop Duration=3, UseLoopDelay=true)
   &SpawnBurst_Instantaneous(Use Spawn Probability=false, UseLoopCountLimit=false)
   &SpawnBurst_Instantaneous(Spawn Count=1, Spawn Group=1, Spawn Time=0.05, Use Spawn Probability=false) {module=SpawnBurst_Instantaneous004}
  ]
  particleSpawnScript [
   &InitialMeshOrientation(Orientation Axis=<0 0 1>, Orientation Vector=<0 0 1>, Rotation=<0 0.25 0>, Use Orientation Vector=true, Use Rotation Vector=true)
   &InitialMeshOrientation(
    Orientation Axis=<0 0 0>
    Orientation Vector=<0 0 0>
    Rotation=&RandomRangeVector(Minimum=<0 0 0>, Maximum=<1 0 0>)
    Use Orientation Vector=false
    Use Rotation Vector=true
   ) {module=InitialMeshOrientation001}
   /SetVariables(Particles.VisibilityTag=&SpawnGroup()) {isSetParameters=true}
   &InitializeParticle(
    Color=&MaskLinearColorBySpawnGroup(Masked Linear Color=<rgba 1.2 1.2 1.2 1>, Passthrough Linear Color=<rgba 0 1.75 3 0.6>)
    Lifetime=&MaskFloatBySpawnGroup(Masked Float=0.3, Passthrough Float=0.15)
    Mesh Scale Mode=::NewEnumerator3
    Mesh Uniform Scale=&MaskFloatBySpawnGroup(Masked Float=1.5, Passthrough Float=1)
    Position Offset:NiagaraPosition=$User.Offset
    Sprite Size Mode=::NewEnumerator3
    Uniform Sprite Size=&MaskFloatBySpawnGroup(Masked Float=10, Passthrough Float=700)
    UsePositionOffset=true
    Write Material Random=true
   ) {module=InitializeParticle001}
   &CameraOffset(Camera Offset Amount=&MaskFloatBySpawnGroup(Masked Float=0, Passthrough Float=50))
  ]
  particleUpdateScript [
   &ParticleState()
   &ScaleColor(
    Scale Mode=::NewEnumerator1
    Scale RGBA=/Game/Effects/Niagara/Modules/MergeEmitters/MergeHelper_Vector4Switch(
     0:Vector4f=&MakeVector4_FromLinearColor(LinearColor=&ColorFromCurve(ColorCurve=DataInterface() {propertyValues {
         AlphaCurve {keys [<at 0 1>,<at 0.2388 1>,<at 1 0>]}
         BlueCurve {keys [<at 0 1>,<at 0.168 1>,<at 0.365 0.7>,<at 1 0.6>]}
         GreenCurve {keys [<at 0 0.6667>,<at 0.168 0.5833>,<at 0.365 0.35>,<at 1 0.25>]}
         RedCurve {keys [<at 0 0>,<at 0.168 0>,<at 0.365 0>,<at 1 0>]}
        }}))
     1:Vector4f=<1 1 1 1>
     2:Vector4f=?
     3:Vector4f=?
     4:Vector4f=?
     5:Vector4f=?
     6:Vector4f=?
     7:Vector4f=?
     8:Vector4f=?
     9:Vector4f=?
     Index:NiagaraInt32=&SpawnGroup()
     Num:MergeHelper_SwitcherEnum=::NewEnumerator0
    )
    ScaleRGBA=true
   ) {module=ScaleColor002}
   &ScaleColor(
    Scale Alpha=&MaskFloatBySpawnGroup(
     Masked Float=&FloatFromCurve(FloatCurve=DataInterface() {propertyValues {Curve {keys [{interpMode=RCIM_Cubic, tangentMode=RCTM_User, time=0.5, value=1}, {interpMode=RCIM_Cubic, tangentMode=RCTM_User, time=1, value=0}]}}})
     Passthrough Float=&FloatFromCurve(FloatCurve=DataInterface() {propertyValues {Curve {keys [
          <at 0 0 RCIM_Cubic>
          {interpMode=RCIM_Cubic, tangentMode=RCTM_Break, time=0.4, value=1}
          {interpMode=RCIM_Cubic, tangentMode=RCTM_Break, time=1, value=0}
         ]}}})
    )
    Scale RGB=&MaskVectorBySpawnGroup(
     Masked Vector=&VectorFromFloat(Value=&FloatFromCurve(FloatCurve=DataInterface() {propertyValues {Curve {keys [
           {interpMode=RCIM_Cubic, tangentMode=RCTM_Break, time=0, value=10}
           {arriveTangent=-6.6072, interpMode=RCIM_Cubic, leaveTangent=-6.6072, tangentMode=RCTM_User, time=0.15, value=3}
           {interpMode=RCIM_Cubic, tangentMode=RCTM_Break, time=1, value=1}
          ]}}}))
     Passthrough Vector=<1 1 1>
    )
    ScaleA=true
    ScaleRGB=true
   ) {module=ScaleColor005}
   &ScaleMeshSize(Scale Factor=&VectorFromFloat(Value=&FloatFromCurve(FloatCurve=DataInterface() {propertyValues {Curve {keys [
          {interpMode=RCIM_Cubic, tangentMode=RCTM_Break, time=0, value=0.2}
          {arriveTangent=2.9365, interpMode=RCIM_Cubic, leaveTangent=2.9365, tangentMode=RCTM_User, time=0.1, value=1}
          {interpMode=RCIM_Cubic, tangentMode=RCTM_Break, time=1, value=2}
         ]}}})))
   &ScaleSpriteSize(Uniform Curve Sprite Scale=DataInterface() {propertyValues {Curve {keys [
        <at 0 0 RCIM_Cubic>
        {interpMode=RCIM_Cubic, tangentMode=RCTM_Break, time=0.4, value=1}
        {interpMode=RCIM_Cubic, tangentMode=RCTM_Break, time=1, value=0}
       ]}}})
   &DynamicMaterialParameters(
    Index 0 Param 1=&FloatFromCurve(FloatCurve=DataInterface() {propertyValues {Curve {keys [{arriveTangent=-2.1015, interpMode=RCIM_Cubic, leaveTangent=-2.1015, tangentMode=RCTM_User, time=0, value=0.5}, <at 1 -0.15>]}}})
    Index 0 Param 2=&FloatFromCurve(FloatCurve=DataInterface() {propertyValues {Curve {keys [{arriveTangent=-0.3464, interpMode=RCIM_Cubic, leaveTangent=-0.3464, tangentMode=RCTM_User, time=0, value=1}, <at 1 0.9>]}}})
    Index 0 Param 3=&FloatFromCurve(FloatCurve=DataInterface() {propertyValues {Curve {keys [
         {interpMode=RCIM_Cubic, tangentMode=RCTM_Break, time=0, value=0}
         {arriveTangent=0.4478, interpMode=RCIM_Cubic, leaveTangent=0.4478, tangentMode=RCTM_User, time=0.15, value=1}
         {interpMode=RCIM_Cubic, tangentMode=RCTM_Break, time=1, value=1.2}
        ]}}})
    Module.Param0WriteEnabled=true
    Module.Param1WriteEnabled=true
    Module.Param2WriteEnabled=true
    Module.Param3WriteEnabled=false
   )
   /SetVariables(Particles.Vector 4:Vector4f=<1 0.5 0 1>) {isSetParameters=true}
  ]
  rendererClasses [&NiagaraMeshRendererProperties, &NiagaraSpriteRendererProperties]
  renderers [
   {rendererClass=&NiagaraMeshRendererProperties, rendererIndex=0}
   {rendererClass=&NiagaraSpriteRendererProperties, rendererIndex=1}
  ]
 }
 EnvChecker {
  emitterUpdateScript [
   &EmitterState(Life Cycle Mode=::NewEnumerator1, Loop Behavior=::NewEnumerator1, Loop Delay=0.15, Loop Duration Mode=::NewEnumerator1, UseLoopDelay=true)
   &SpawnRate(SpawnRate=&Multiply_Float(A=10, B=$User.EnableValue), Use Spawn Probability=false)
   &self:ScratchModule_06(Simple Counter:NiagaraDataInterfaceSimpleCounter=$User.Simple Counter) {module=ScratchModule_06}
  ]
  particleSpawnScript [
   &InitializeParticle(
    Color Mode=::NewEnumerator0
    Lifetime=&RandomRangeFloat(Minimum=0.2, Maximum=0.3)
    Sprite Size Mode=::NewEnumerator3
    UsePositionOffset=false
    Write Material Random=true
   )
   &ShapeLocation(
    Handle Radius=5
    Large Radius=75.0538
    MaskSpawnGroup=false
    Non Uniform Scale=<1 1.5 1>
    Offset Coordinate Space=::Local
    Offset Mode=::NewEnumerator0
    Override Seed=false
    Rotation Mode=::NewEnumerator2
    Shape Primitive=::NewEnumerator3
    UseOwnerScale=true
   )
   &AddVelocity(Rotation Coordinate Space=::World, Velocity=<0 0 -2000>, Velocity Speed Scale=0)
   /SetVariables() {isSetParameters=true}
   /SetVariables(Particles.Position=&AddVectorToPosition(Position=$Particles.Position, Vector=<0 0 50>)) {isSetParameters=true}
  ]
  particleUpdateScript [
   &ParticleState()
   &self:ClearArray(Simple Counter:NiagaraDataInterfaceSimpleCounter=$User.Simple Counter, Vector 4 Array:NiagaraDataInterfaceArrayFloat4=$User.Vector 4 Array) {module=ClearArray}
   &self:ScratchModule_04(Offset=100, TraceDir:Vector3f=<0 0 -1>, TraceDist=500, bIsFrontEnd:NiagaraBool=$User.bIsFrontEndPreview) {module=ScratchModule_04}
   &CurlNoiseForce(Noise Strength=300, Pan Noise=false)
   &SolveForcesAndVelocity(Clamp Velocity=false, Limit Acceleration=false)
   &self:ScratchModule_01(
    Float Array:NiagaraDataInterfaceArrayFloat=$User.RandomArray
    ParticleRDM:NiagaraFloat=&RandomRangeFloat(Minimum=0, Maximum=1)
    Position:NiagaraPosition=$Particles.CollisionPos
    Vector 4 Array:NiagaraDataInterfaceArrayFloat4=$User.Vector 4 Array
   ) {module=ScratchModule_01}
  ]
  rendererClasses [&NiagaraSpriteRendererProperties]
  renderers [{rendererClass=&NiagaraSpriteRendererProperties}]
 }
 FlaresAndEnergyRingAndPL {
  emitterUpdateScript [
   &EmitterState(
    Enable Distance Culling=true
    Life Cycle Mode=::NewEnumerator1
    Loop Behavior=::NewEnumerator1
    Loop Duration Mode=::NewEnumerator1
    MaxDistance=10000
    Module.bUseMaxDistance=true
    Module.bUseMinDistance=false
    Scalability Mode=::NewEnumerator1
    UseLoopDelay=false
   )
   &SpawnBurst_Instantaneous(Spawn Count=1, Use Spawn Probability=false)
   &SpawnBurst_Instantaneous(Spawn Count=1, Spawn Group=1, Use Spawn Probability=false) {module=SpawnBurst_Instantaneous001}
   &SpawnBurst_Instantaneous(Spawn Group=2, Spawn Time=0.5, Use Spawn Probability=false, UseLoopCountLimit=false) {module=SpawnBurst_Instantaneous002}
   &SpawnBurst_Instantaneous(Spawn Group=3, Use Spawn Probability=false, UseLoopCountLimit=false) {module=SpawnBurst_Instantaneous003}
  ]
  particleSpawnScript [
   &DynamicMaterialParameters(
    Index 0 Param 1=&RandomRangeFloat(Minimum=0.16, Maximum=0.35, Evaluation Type=::NewEnumerator1)
    Index 0 Param 2=&RandomRangeFloat(Minimum=1, Maximum=2, Evaluation Type=::NewEnumerator1)
    Index 0 Param 3=1
    Index 0 Param 4=1
    Module.Param0WriteEnabled=true
    Module.Param1WriteEnabled=true
    Module.Param2WriteEnabled=true
    Module.Param3WriteEnabled=true
   )
   &CameraOffset(Camera Offset Amount=10) {module=CameraOffset001}
   /SetVariables(Particles.VisibilityTag=&SpawnGroup()) {isSetParameters=true}
   &InitializeParticle(
    Color=&MergeHelper_ColorSwitch(
     0=<rgba 0.2453 0.9334 1.25 1>
     1=<rgba 0 0.5 1 0.23>
     2=<rgba 0 0.5 1 0.5>
     3=<rgba 0.02 0.11 0.2 0.7>
     Index=&SpawnGroup()
     Num=::NewEnumerator2
    )
    Lifetime=&MergeHelper_FloatSwitch(
     0=&RandomRangeFloat(Minimum=2.7, Maximum=3.2)
     1=0.7
     2=0.2
     3=0.8
     Index=&SpawnGroup()
     Num=::NewEnumerator2
    )
    Mesh Scale Mode=::NewEnumerator3
    Mesh Uniform Scale=6.6199
    Position Offset=&MergeHelper_VectorSwitch(
     0=Linked() {linkedVariable {name=User.Offset, type=NiagaraPosition}}
     1=Linked() {linkedVariable {name=User.Offset, type=NiagaraPosition}}
     2=<0 0 -61>
     3=<0 0 30>
     Index=&SpawnGroup()
     Num=::NewEnumerator2
    )
    Sprite Size=&MergeHelper_Vector2Switch(0=<120 120>, 1=<6770 2000>, 2=<6.6 6.6>, 3=<419 419>, 4=?, Index=&SpawnGroup(), Num=::NewEnumerator2)
    Sprite Size Mode=::NewEnumerator4
    UsePositionOffset=true
    Write Material Random=true
   ) {module=InitializeParticle004}
   &InitialMeshOrientation(Mesh Axis To Orient=<0 0 1>, Mesh Orientation Mode=::NewEnumerator2, Use Rotation Vector=false)
  ]
  particleUpdateScript [
   &ParticleState(Kill Particles When Lifetime Has Elapsed=false, Let Infinitely Lived Particles Die When Emitter Deactivates=true)
   &SolveForcesAndVelocity(Clamp Velocity=false, Limit Acceleration=false)
   /SetVariables(Particles.LightBrightness:NiagaraFloat=&Multiply_Float(A=1500, B=$User.EnableValue)) {isSetParameters=true}
   /SetVariables(Particles.LightColorInit:LinearColor=&MultiplyLinearColor(
     Linear Color=&MultiplyLinearColor(
      Linear Color=&LerpLinearColors(StartColor=<rgba 1 4 8 1>, EndColor=<rgba 10 15 30 1>, LerpFactor=$User.Open)
      Scale Factor=<1 1 1 1>
     )
     Scale Factor=&MakeVector4(X=$User.EnableValue, Y=$User.EnableValue, Z=$User.EnableValue, W=1)
    )) {isSetParameters=true}
   /SetVariables(Particles.LightColor=&MultiplyLinearColor(
     Linear Color=$Particles.LightColorInit
     Scale Factor=&MakeVector4_FromLinearColor(LinearColor=<rgba 0.2941 1.7157 2 1>)
    )) {isSetParameters=true}
   /SetVariables(Particles.LightExponent=12) {isSetParameters=true}
   /SetVariables(Particles.LightPosition:NiagaraPosition=<0 0 40>, Particles.LightRadius=500) {isSetParameters=true}
   &CameraOffset(Camera Offset Amount=&MaskFloatBySpawnGroup(Masked Float=100, Passthrough Float=0))
   &ScaleColor(
    Scale Alpha=&MergeHelper_FloatSwitch(
     0=$User.EnableValue
     1=$User.EnableValue
     2=&FloatFromCurve(FloatCurve=DataInterface() {propertyValues {Curve {keys [<at 0.5 0>,<at 1 1>]}}}, CurveIndex=$Emitter.Age, Scale Curve=$User.EnableValue)
     3=$User.EnableValue
     Index=&SpawnGroup()
     Num=::NewEnumerator2
    )
    Scale RGB=&MergeHelper_VectorSwitch(
     0=&VectorFromFloat(Value=&FloatFromCurve(FloatCurve=DataInterface() {propertyValues {Curve {keys [<at 0 10>,<at 0.5 1>]}}}, CurveIndex=$Emitter.Age))
     1=&VectorFromFloat(Value=&FloatFromCurve(FloatCurve=DataInterface() {propertyValues {Curve {keys [<at 0 10>,<at 0.5 1>]}}}, CurveIndex=$Emitter.Age))
     2=&VectorFromFloat(Value=&RandomRangeFloat(Minimum=0.6, Maximum=1, Evaluation Type=::NewEnumerator1))
     3=<1 1 1>
     Index=&SpawnGroup()
     Num=::NewEnumerator2
    )
    ScaleA=true
    ScaleRGB=true
   ) {module=ScaleColor005}
   &ScaleSpriteSize(
    Scale Factor=&MergeHelper_Vector2Switch(
     0=&Vector2DFromFloat(Value=&FloatFromCurve(
       FloatCurve=DataInterface() {propertyValues {Curve {keys [<at 0 1>,<at 10000 3>]}}}
       CurveIndex=$Engine.Owner.LODDistance
       Scale Curve=&FloatFromCurve(FloatCurve=DataInterface() {propertyValues {Curve {keys [<at 0 0>,<at 0.2 1>]}}}, CurveIndex=$Emitter.Age)
      ))
     1=&Multiply_Vector2D(
      A=&Vector2DFromFloat(Value=&Multiply_Float(
        A=&RandomRangeFloat(Minimum=0.8, Maximum=1, Evaluation Type=::NewEnumerator1)
        B=&FloatFromCurve(FloatCurve=DataInterface() {propertyValues {Curve {keys [<at 0 0>,<at 0.2 1>]}}}, CurveIndex=$Emitter.Age)
       ))
      B=&Multiply_Vector2D(
       A=&Vector2DFromFloat(Value=0.6)
       B=&Vector2DFromFloat(Value=&Lerp_Float(A=0.7, B=1.1, Alpha=0.5))
      )
     )
     2=<1 1>
     3=<1 1>
     4=?
     Index=&SpawnGroup()
     Num=::NewEnumerator2
    )
    Scale Sprite Size Mode=::NewEnumerator2
   ) {module=ScaleSpriteSize006}
   /SetVariables(Particles.MergeDP1:Vector4f=&MergeHelper_Vector4Switch(
     0:Vector4f=&MakeVector4(
      X=&RandomRangeFloat(Minimum=0.16, Maximum=0.35)
      Y=&RandomRangeFloat(Minimum=1, Maximum=2)
      Z=1
      W=1
     )
     1:Vector4f=<1 0.5 0 1>
     2:Vector4f=<0 0 0 0>
     3:Vector4f=<1 1 1 1>
     4:Vector4f=?
     5:Vector4f=?
     6:Vector4f=?
     7:Vector4f=?
     8:Vector4f=?
     9:Vector4f=?
     Index:NiagaraInt32=&SpawnGroup()
     Num:MergeHelper_SwitcherEnum=::NewEnumerator2
    )) {isSetParameters=true}
  ]
  rendererClasses [
   &NiagaraSpriteRendererProperties
   &NiagaraMeshRendererProperties
   &NiagaraLightRendererProperties
  ]
  renderers [
   {rendererClass=&NiagaraSpriteRendererProperties}
   {rendererClass=&NiagaraSpriteRendererProperties, rendererIndex=1}
   {rendererClass=&NiagaraMeshRendererProperties, rendererIndex=2}
   {rendererClass=&NiagaraSpriteRendererProperties, rendererIndex=3}
   {rendererClass=&NiagaraLightRendererProperties, rendererIndex=4}
  ]
 }
 LinkLightning {
  emitterUpdateScript [
   &EmitterState(Life Cycle Mode=::NewEnumerator1, Loop Behavior=::NewEnumerator1, Loop Duration Mode=::NewEnumerator1, UseLoopDelay=false)
   &SpawnBurst_Instantaneous(Spawn Count=30, Use Spawn Probability=false, UseLoopCountLimit=false)
   /SetVariables(Emitter.TipGlowStr:NiagaraFloat=&FN_FrontEndFloatModifier(Gameplay=3.5, Front End=2.7)) {isSetParameters=true}
  ]
  particleSpawnScript [
   &FN_SetSpawnGroup(Inclusivity=true, Normalize Spawn Group=true, Number of Spawn Groups=6, Particles Per Group=5)
   &InitializeParticle(
    Color=<rgba 2 2 2 1>
    Lifetime=5
    Ribbon Width=10
    Ribbon Width Mode=::NewEnumerator1
    Sprite Size Mode=::NewEnumerator3
    Uniform Sprite Size=15
    UsePositionOffset=true
    Write Material Random=true
   )
  ]
  particleUpdateScript [
   &ParticleState(Kill Particles When Lifetime Has Elapsed=false, Let Infinitely Lived Particles Die When Emitter Deactivates=true)
   &self:Getpos(
    MaxTarget=6
    ParticlePerRibbon=5
    RandomArray:NiagaraDataInterfaceArrayFloat=$User.RandomArray
    Vector 4 Array:NiagaraDataInterfaceArrayFloat4=$User.Vector 4 Array
   ) {enabled=false, module=Getpos}
   &self:Getpos_01(
    Angle=90
    Cone Distribution=1
    InnerAngle=45
    MaxTarget=6
    NormalizedSpawngroup:NiagaraFloat=$Particles.Normalized Spawn Group
    O:NiagaraPosition=$User.Offset
    ParticlePerRibbon=5
    Radius=5
    RandomArray:NiagaraDataInterfaceArrayFloat=$User.RandomArray
    SpawnGroup:NiagaraInt32=$Particles.SpawnGroup
    V:Vector3f=<0 0 0.5>
    Vector 4 Array:NiagaraDataInterfaceArrayFloat4=$User.Vector 4 Array
   ) {module=Getpos_01}
   &ScaleColor(
    Scale Alpha=&FloatFromCurve(
     FloatCurve=DataInterface() {propertyValues {Curve {keys [
         {tangentMode=RCTM_User, time=0, value=1}
         {arriveTangent=-0.0405, interpMode=RCIM_Cubic, leaveTangent=-0.0405, tangentMode=RCTM_User, time=0.7009, value=1}
         {arriveTangent=-8.0089, interpMode=RCIM_Cubic, leaveTangent=-8.0089, tangentMode=RCTM_User, time=1, value=0}
        ]}}}
     CurveIndex=$Particles.W
    )
    ScaleA=true
    ScaleRGB=true
   )
   &self:ScratchModule(EndPos:NiagaraPosition=$Particles.EndPos, StartPos:NiagaraPosition=$Engine.Emitter.SimulationPosition) {module=ScratchModule}
   /SetVariables(Particles.GrowthV:NiagaraFloat=&FloatFromCurve(FloatCurve=DataInterface() {propertyValues {Curve {keys [<at 0 0>,<at 0.3 1>]}}}, CurveIndex=$Particles.W)) {isSetParameters=true}
   &DynamicMaterialParameters(
    Index 0 Param 1=$Particles.ScratchModule.X
    Index 0 Param 2=$Particles.ScratchModule.Y
    Index 0 Param 3=$Particles.ScratchModule.Z
    Index 0 Param 4=$Particles.ScratchModule.W
    Index 1 Param 1=$Particles.RDM
    Index 1 Param 2=$Particles.GrowthV
    Index 1 Param 3=&FloatFromCurve(FloatCurve=DataInterface() {propertyValues {Curve {keys [<at 0 0>,<at 0.5 1>,<at 0.6 1>,<at 1 0>]}}}, CurveIndex=$Particles.W)
    Index 1 Param 4=&Add_Float(
     A=&FloatFromCurve(
      FloatCurve=DataInterface() {propertyValues {Curve {keys [{tangentMode=RCTM_User, time=0, value=-1}, {tangentMode=RCTM_User, time=1, value=0.7}]}}}
      CurveIndex=$Particles.W
     )
     B=&FloatFromCurve(
      FloatCurve=DataInterface() {propertyValues {Curve {keys [
          {interpMode=RCIM_Cubic, tangentMode=RCTM_Break, time=0, value=-1}
          {interpMode=RCIM_Cubic, tangentMode=RCTM_Break, time=0.1, value=0}
          {interpMode=RCIM_Cubic, tangentMode=RCTM_Break, time=0.9, value=0}
          <at 1 -1>
         ]}}}
      CurveIndex=$Particles.RibbonLinkOrder
      Scale Curve=0.3
     )
    )
    Module.Index0Param0Write=true
    Module.Index0Param1Write=true
    Module.Index0Param2Write=true
    Module.Index0Param3Write=true
    Module.Index1Param0Write=true
    Module.Index1Param1Write=true
    Module.Index1Param2Write=true
    Module.Index1Param3Write=true
    Write Parameter Index 1=true
   )
   &self:BendRibbon(
    BigNoiseMask:NiagaraFloat=&FloatFromCurve(
     FloatCurve=DataInterface() {propertyValues {Curve {keys [
         {arriveTangent=2.6981, interpMode=RCIM_Cubic, leaveTangent=2.6981, tangentMode=RCTM_User, time=0, value=0}
         {interpMode=RCIM_Cubic, tangentMode=RCTM_User, time=0.5, value=1}
         {arriveTangent=-0.3624, interpMode=RCIM_Cubic, tangentMode=RCTM_Break, time=1, value=0}
        ]}}}
     CurveIndex=$Particles.RibbonLinkOrder
    )
    Offset:NiagaraFloat=&FloatFromCurve(
     FloatCurve=DataInterface() {propertyValues {Curve {keys [
         {arriveTangent=-0.0524, interpMode=RCIM_Cubic, leaveTangent=-0.0524, tangentMode=RCTM_User, time=0, value=0}
         {interpMode=RCIM_Cubic, tangentMode=RCTM_User, time=0.5, value=1}
         {arriveTangent=-4.0426, tangentMode=RCTM_User, time=1, value=0}
        ]}}}
     CurveIndex=$Particles.RibbonLinkOrder
     Scale Curve=&FloatFromCurve(
      FloatCurve=DataInterface() {propertyValues {Curve {keys [{interpMode=RCIM_Cubic, leaveTangent=2.3336, tangentMode=RCTM_Break, time=0, value=0.2}, <at 1 1>]}}}
      CurveIndex=$Particles.W
      Scale Curve=&Multiply_Float(
       A=&RemapRange(Input Value To Remap=$Particles.RDM, Input Min=0, Input Max=1, Output Min=0, Output Max=1)
       B=-30
      )
     )
    )
    RibbonLength:NiagaraFloat=$Particles.RibbonLength
    YBigNoise:NiagaraFloat=&Waveform(
     Global Input Position=&Multiply_Float(A=$Particles.RDM, B=10)
     Global Phase=&Add_Float(
      A=&Multiply_Float(A=$Particles.RibbonLinkOrder, B=2)
      B=&Multiply_Float(A=$Emitter.Age, B=-2)
     )
     "Waveform [1]"=::NewEnumerator12
     "[1] Frequency"=1
     "[1] Amplitude Min/Max"=<-1 1>
     "[1] Scale"=0
    )
    ZBigNoise:NiagaraFloat=&Waveform(
     Global Input Position=&Multiply_Float(A=$Particles.RDM, B=10)
     Global Phase=&Add_Float(
      A=&Multiply_Float(A=$Particles.RibbonLinkOrder, B=3)
      B=&Multiply_Float(A=$Emitter.Age, B=2)
     )
     "Waveform [1]"=::NewEnumerator12
     "[1] Frequency"=1
     "[1] Amplitude Min/Max"=<-1 0>
     "[1] Scale"=15
    )
   ) {module=BendRibbon}
   &ScaleRibbonWidth(Ribbon Width Scale=&FloatFromCurve(
     FloatCurve=DataInterface() {propertyValues {Curve {keys [
         {interpMode=RCIM_Cubic, tangentMode=RCTM_Break, time=0, value=0.6}
         {interpMode=RCIM_Cubic, tangentMode=RCTM_Break, time=0.1, value=0.4}
         {interpMode=RCIM_Cubic, tangentMode=RCTM_Break, time=0.9, value=0.4}
         <at 1 1>
        ]}}}
     CurveIndex=$Particles.RibbonLinkOrder
     Scale Curve=&FloatFromCurve(
      FloatCurve=DataInterface() {propertyValues {Curve {keys [
          {tangentMode=RCTM_User, time=0, value=1}
          {arriveTangent=-0.0405, interpMode=RCIM_Cubic, leaveTangent=-0.0405, tangentMode=RCTM_User, time=0.5, value=1}
          {arriveTangent=-4.0987, interpMode=RCIM_Cubic, leaveTangent=-4.0987, tangentMode=RCTM_User, time=1, value=0}
         ]}}}
      CurveIndex=$Particles.W
     )
    ))
   &ScaleRibbonWidth(Ribbon Width Scale=&RemapRange(Input Value To Remap=$Particles.RDM, Input Min=0, Input Max=1, Output Min=0.7, Output Max=2)) {module=ScaleRibbonWidth001}
   &ScaleColor(
    Scale RGB=&VectorFromFloat(Value=&MakeCustomFloatFromBool(
      Bool=$User.bIsFrontEndPreview
      True Float=&FloatFromCurve(
       FloatCurve=DataInterface() {propertyValues {Curve {keys [<at 0 2.7>,<at 0.3 1>]}}}
       CurveIndex=$Particles.W
       Scale Curve=&RandomRangeFloat(Minimum=0.4, Maximum=1, Evaluation Type=::NewEnumerator1)
      )
      False Float=&FloatFromCurve(
       FloatCurve=DataInterface() {propertyValues {Curve {keys [<at 0 7>,<at 0.3 1>]}}}
       CurveIndex=$Particles.W
       Scale Curve=&RandomRangeFloat(Minimum=0.4, Maximum=1, Evaluation Type=::NewEnumerator1)
      )
     ))
    ScaleA=true
    ScaleRGB=true
   ) {module=ScaleColor001}
   /SetVariables(Particles.RibbonW2:NiagaraFloat=&Multiply_Float(A=$Particles.RibbonWidth, B=3)) {isSetParameters=true}
   &DoOnce(
    Reset:NiagaraBool=&SetBoolByFloatComparison(A=$Particles.W, B=0, Comparison Type=::NewEnumerator7)
    Trigger Condition:NiagaraBool=&Boolean_AndOperation(
     Bool1=&SetBoolByFloatComparison(A=$Particles.RibbonLinkOrder, B=1, Comparison Type=::NewEnumerator2)
     Bool2=&SetBoolByFloatComparison(A=$Particles.GrowthV, B=0.9)
    )
   )
   &GenerateLocationEvent(
    Event Generation Enabled:NiagaraBool=&Boolean_AndOperation(
     Bool1=&Boolean_AndOperation(
      Bool1=&Boolean_AndOperation(Bool1=$Particles.bIsHead, Bool2=&SetBoolByFloatComparison(A=$Particles.GrowthV, B=0.99))
      Bool2=&SetBoolByFloatComparison(A=$Particles.W, B=0.7, Comparison Type=::NewEnumerator7)
     )
     Bool2=$Particles.RibbonValid
    )
    Event Send Rate=15
    "Float to Send as Float 2 (RandomNormalizedFloat)":NiagaraFloat=$Particles.RDM
    Use Event Probability=false
    UseEventDelay=false
   )
   &FN_Sprite_Attributes(
    Color=&MultiplyLinearColor(
     Linear Color=<rgba 0.05 0.35 0.5 1>
     Scale Factor=&MakeFloat4FromVectorAndFloat(
      "Vector (RGB)"=<15 15 15>
      "Float (Alpha)"=&FloatFromCurve(
       FloatCurve=DataInterface() {propertyValues {Curve {keys [
           <at 0.29 0 RCIM_Constant>
           <at 0.3 1>
           {arriveTangent=-6.0095, interpMode=RCIM_Cubic, leaveTangent=-2.1556, tangentMode=RCTM_Break, time=0.6, value=0}
          ]}}}
       CurveIndex=$Particles.W
       Scale Curve=&Multiply_Float(
        A=&RemapRange(Input Value To Remap=$Particles.RDM, Input Min=0, Input Max=1, Output Min=0.3, Output Max=0.9)
        B=&Multiply_Float(
         A=&FNTimeOfDayScalar(Night Scale=0.7, Morning Scale=1, Evening Scale=1)
         B=&FN_FrontEndFloatModifier(Gameplay=1, Front End=0.5)
        )
       )
      )
     )
    )
    Create Color Attribute=true
    Create SpriteSize Attribute=true
    SpriteSize=&Multiply_Vector2D_ByFloat(
     Float=&FloatFromCurve(
      FloatCurve=DataInterface() {propertyValues {Curve {keys [<at 0.3 1>,<at 0.5 2>]}}}
      CurveIndex=$Particles.W
      Scale Curve=&RemapRange(Input Value To Remap=$Particles.RDM, Input Min=0, Input Max=1, Output Min=0.3, Output Max=1)
     )
     Vector2D=&MakeCustomVector2DFromBool(Bool=$Particles.bIsHead, True Vector2D=<20 20>, False Vector2D=<0 0>)
    )
    Write to Module Namespace=true
   )
   &FN_Sprite_Attributes(
    Color=&MultiplyLinearColor(
     Linear Color=<rgba 0.04 0.28 0.4 0.8>
     Scale Factor=&MakeFloat4FromVectorAndFloat(
      "Vector (RGB)"=<1 1 1>
      "Float (Alpha)"=&FloatFromCurve(
       FloatCurve=DataInterface() {propertyValues {Curve {keys [
           <at 0.29 0 RCIM_Constant>
           <at 0.3 1>
           {arriveTangent=-6.0095, interpMode=RCIM_Cubic, leaveTangent=-2.1556, tangentMode=RCTM_Break, time=0.6, value=0}
          ]}}}
       CurveIndex=$Particles.W
       Scale Curve=&RemapRange(Input Value To Remap=$Particles.RDM, Input Min=0, Input Max=1, Output Min=1, Output Max=1.5)
      )
     )
    )
    Create Color Attribute=true
    Create SpriteSize Attribute=true
    SpriteSize=&Multiply_Vector2D_ByFloat(
     Float=&FloatFromCurve(
      FloatCurve=DataInterface() {propertyValues {Curve {keys [<at 0.3 1>,<at 0.5 2>]}}}
      CurveIndex=$Particles.W
      Scale Curve=&RemapRange(Input Value To Remap=$Particles.RDM, Input Min=0, Input Max=1, Output Min=0.3, Output Max=1)
     )
     Vector2D=&MakeCustomVector2DFromBool(Bool=$Particles.bIsHead, True Vector2D=<60 60>, False Vector2D=<0 0>)
    )
    Write to Module Namespace=true
   ) {module=FN_Sprite_Attributes002}
   &FN_Sprite_Attributes(
    Color=&MultiplyLinearColor(
     Linear Color=<rgba 0.3 1.4333 2 1>
     Scale Factor=&MakeFloat4FromVectorAndFloat(
      "Vector (RGB)"=<1 1 1>
      "Float (Alpha)"=&FloatFromCurve(
       FloatCurve=DataInterface() {propertyValues {Curve {keys [
           <at 0.29 0 RCIM_Constant>
           <at 0.3 1>
           <at 0.4 0.3>
           {arriveTangent=-6.0095, interpMode=RCIM_Cubic, leaveTangent=-2.1556, tangentMode=RCTM_Break, time=0.6, value=0}
          ]}}}
       CurveIndex=$Particles.W
       Scale Curve=&RemapRange(Input Value To Remap=$Particles.RDM, Input Min=0, Input Max=1, Output Min=0.3, Output Max=0.9)
      )
     )
    )
    Create Color Attribute=true
    Create Sprite Rotation Attribute=true
    Create SpriteSize Attribute=true
    SpriteRotation:NiagaraFloat=$Particles.RDM
    SpriteSize=&Multiply_Vector2D_ByFloat(
     Float=&FloatFromCurve(
      FloatCurve=DataInterface() {propertyValues {Curve {keys [<at 0.3 1>,<at 0.5 2>]}}}
      CurveIndex=$Particles.W
      Scale Curve=&RemapRange(Input Value To Remap=$Particles.RDM, Input Min=0, Input Max=1, Output Min=0.5, Output Max=1.2)
     )
     Vector2D=&MakeCustomVector2DFromBool(Bool=$Particles.bIsHead, True Vector2D=<300 300>, False Vector2D=<0 0>)
    )
    Write to Module Namespace=true
   ) {module=FN_Sprite_Attributes001}
   &FN_Mesh_Attributes(
    Color=&ColorFromCurve(
     ColorCurve=DataInterface() {propertyValues {
       AlphaCurve {keys [<at 0 1 RCIM_Cubic>,<at 1 0 RCIM_Cubic>]}
       BlueCurve {keys [<at 0.3 5 RCIM_Cubic>,<at 0.417 3>,<at 0.8577 1 RCIM_Cubic>]}
       GreenCurve {keys [<at 0.3 2 RCIM_Cubic>,<at 0.417 1.65>,<at 0.8577 0.34 RCIM_Cubic>]}
       RedCurve {keys [<at 0.3 0.5 RCIM_Cubic>,<at 0.417 0.3>,<at 0.8577 0.01 RCIM_Cubic>]}
      }}
     CurveIndex=$Particles.W
    )
    Create Color Attribute=true
    Create Orientation Attribute=true
    Create Scale Attribute=true
    Orientation:Quat4f=&MakeQuatFromAxisAngle(Axis=<0 1 0>, AngleInDegrees=&Multiply_Float(A=$Particles.RDM, B=360))
    Scale=&VectorFromFloat(Value=&MakeCustomFloatFromBool(
      Bool=$Particles.bIsHead
      True Float=&FloatFromCurve(
       FloatCurve=DataInterface() {propertyValues {Curve {keys [<at 0.3 0.5>,<at 0.8 1>]}}}
       CurveIndex=$Particles.W
       Scale Curve=&RemapRange(Input Value To Remap=$Particles.RDM, Input Min=0, Input Max=1, Output Min=0.5, Output Max=1)
      )
      False Float=0
     ))
    Write to Module Namespace=true
   )
   &FN_Mesh_Attributes(
    Color=&MultiplyLinearColor(
     Linear Color=&ColorFromCurve(
      ColorCurve=DataInterface() {propertyValues {
        AlphaCurve {keys [<at 0.29 0>,<at 0.3 0.4>,<at 1 0 RCIM_Cubic>]}
        BlueCurve {keys [<at 0.3 0.7 RCIM_Cubic>,<at 0.6545 0.5>]}
        GreenCurve {keys [<at 0.3 0.4783 RCIM_Cubic>,<at 0.6545 0.35>]}
        RedCurve {keys [<at 0.3 0.035 RCIM_Cubic>,<at 0.6545 0.05>]}
       }}
      CurveIndex=$Particles.W
     )
     Scale Factor=&MakeFloat4FromVectorAndFloat(
      "Vector (RGB)"=&VectorFromFloat(Value=&FN_FrontEndFloatModifier(Gameplay=1, Front End=0.7))
      "Float (Alpha)"=1
     )
    )
    Create Color Attribute=true
    Create Orientation Attribute=true
    Create Scale Attribute=true
    Scale=&VectorFromFloat(Value=&MakeCustomFloatFromBool(
      Bool=$Particles.bIsHead
      True Float=&FloatFromCurve(
       FloatCurve=DataInterface() {propertyValues {Curve {keys [<at 0.299 0>,<at 0.3 1>,<at 0.8 1>]}}}
       CurveIndex=$Particles.W
       Scale Curve=&RemapRange(Input Value To Remap=$Particles.RDM, Input Min=0, Input Max=1, Output Min=3.5, Output Max=4.5)
      )
      False Float=0
     ))
    Write to Module Namespace=true
   ) {module=FN_Mesh_Attributes001}
   &self:StandaloneDP(DP1:Vector4f=<0.5 1 1 1>, DP4:Vector4f=<1 1 1 1>) {module=StandaloneDP}
   &self:StandaloneDP(
    DP1:Vector4f=&MakeVector4(
     X=&FloatFromCurve(
      FloatCurve=DataInterface() {propertyValues {Curve {keys [
          <at 0.29 1>
          {interpMode=RCIM_Cubic, tangentMode=RCTM_Break, time=0.3, value=0}
          {arriveTangent=3.1558, interpMode=RCIM_Cubic, tangentMode=RCTM_Break, time=0.9, value=1}
         ]}}}
      CurveIndex=$Particles.W
     )
     Y=1
     Z=1
     W=1
    )
    DP2:Vector4f=&MakeVector4(
     X=&RemapRange(Input Value To Remap=$Particles.RDM, Input Min=0, Input Max=1, Output Min=60, Output Max=100)
     Y=0
     Z=1
     W=1
    )
    DP4:Vector4f=<1 1 1 1>
   ) {module=StandaloneDP001}
   &CameraOffset(Camera Offset Amount=10)
   /SetVariables(Particles.CO=20) {isSetParameters=true}
   &ScaleColor(
    Scale RGB=&VectorFromFloat(Value=&FloatFromCurve(FloatCurve=DataInterface() {propertyValues {Curve {keys [<at 0.7 1>,<at 1 10>]}}}, CurveIndex=$Particles.W))
    ScaleA=true
    ScaleRGB=true
   ) {enabled=false, module=ScaleColor002}
   /SetVariables(Particles.Color3:LinearColor=&MultiplyLinearColor(
     Linear Color=$Particles.Color
     Scale Factor=&MakeFloat4FromVectorAndFloat(
      "Vector (RGB)"=&VectorFromFloat(Value=&FloatFromCurve(FloatCurve=DataInterface() {propertyValues {Curve {keys [<at 0.7 1>,<at 1 5>]}}}, CurveIndex=$Particles.W, Scale Curve=1.5))
      "Float (Alpha)"=1
     )
    )) {isSetParameters=true}
   &ScaleColor(
    Scale RGB=&VectorFromFloat(Value=&Multiply_Float(
      A=&FNTimeOfDayScalar(Night Scale=0.7, Morning Scale=1, Evening Scale=1)
      B=&FN_FrontEndFloatModifier(Gameplay=1, Front End=0.8)
     ))
    ScaleA=true
    ScaleRGB=true
   ) {module=ScaleColor003}
  ]
  rendererClasses [
   &NiagaraSpriteRendererProperties
   &NiagaraRibbonRendererProperties
   &NiagaraMeshRendererProperties
  ]
  renderers [
   {rendererClass=&NiagaraSpriteRendererProperties}
   {rendererClass=&NiagaraRibbonRendererProperties, rendererIndex=1}
   {rendererClass=&NiagaraRibbonRendererProperties, rendererIndex=2}
   {rendererClass=&NiagaraMeshRendererProperties, rendererIndex=3}
  ]
 }
 LinkSpark {
  emitterUpdateScript [&EmitterState(Life Cycle Mode=::NewEnumerator1, Loop Behavior=::NewEnumerator1, Loop Duration Mode=::NewEnumerator1, UseLoopDelay=false)]
  particleSpawnScript [
   &InitializeParticle(
    Color=<rgba 0.5 6.8333 10 1>
    Lifetime=&RandomRangeFloat(Minimum=0.1, Maximum=0.2)
    Sprite Size Max=<0.6 10>
    Sprite Size Min=<0.4 5>
    Sprite Size Mode=::NewEnumerator2
    UsePositionOffset=false
    Write Material Random=true
   )
   &AddVelocity(
    Cone Angle=90
    Cone Axis=<0 0 1>
    Inner Cone Angle=80
    Module.UseConeFalloff=false
    Module.UseVelDistribution=false
    Override Seed=false
    Rotation Coordinate Space=::World
    Velocity Mode=::NewEnumerator2
    Velocity Speed=&RandomRangeFloat(Minimum=150, Maximum=200)
   )
  ]
  particleUpdateScript [
   &ParticleState()
   &ScaleSpriteSize(
    Scale Factor=&MakeVector2D(
     X=1
     Y=&FloatFromCurve(FloatCurve=DataInterface() {propertyValues {Curve {keys [
          {interpMode=RCIM_Cubic, tangentMode=RCTM_Break, time=0.3, value=1}
          {arriveTangent=-1.0787, interpMode=RCIM_Cubic, tangentMode=RCTM_Break, time=1, value=0}
         ]}}})
    )
    Scale Sprite Size Mode=::NewEnumerator2
   )
   &GravityForce(Gravity=<0 0 -500>)
   &FN_Sprite_Attributes(
    Color=&MultiplyLinearColor(Linear Color=$Particles.Color, Scale Factor=<1 1 1 0.02>)
    Create Color Attribute=true
    Create SpriteSize Attribute=true
    SpriteSize=<15 40>
    Write to Module Namespace=true
   )
   &SolveForcesAndVelocity(Clamp Velocity=false, Limit Acceleration=false)
   &ScaleColor(
    Scale Alpha=&FloatFromCurve(FloatCurve=DataInterface() {propertyValues {Curve {keys [
         {interpMode=RCIM_Cubic, tangentMode=RCTM_Break, time=0, value=1}
         {arriveTangent=-1.9691, interpMode=RCIM_Cubic, tangentMode=RCTM_Break, time=1, value=0}
        ]}}})
    ScaleA=true
    ScaleRGB=true
   )
  ]
  rendererClasses [&NiagaraSpriteRendererProperties]
  renderers [
   {rendererClass=&NiagaraSpriteRendererProperties}
   {rendererClass=&NiagaraSpriteRendererProperties, rendererIndex=1}
  ]
 }
 M001 {
  emitterUpdateScript [
   &EmitterState(Life Cycle Mode=::NewEnumerator1, Loop Behavior=::NewEnumerator1, Loop Duration Mode=::NewEnumerator1, UseLoopDelay=false)
   &SpawnBurst_Instantaneous(Spawn Count=5, Use Spawn Probability=false, UseLoopCountLimit=false) {module=SpawnBurst_Instantaneous005}
  ]
  particleSpawnScript [
   &PartitionParticles(Number Of Partitions=5)
   &InitializeParticle(
    Color=&MergeHelper_ColorSwitch(
     0=<rgba 0 0.4167 1 0.5>
     1=<rgba 0 3.3333 10 1>
     2=<rgba 0 0.3333 0.5 0.5>
     3=<rgba 0 0.5833 1 1>
     4=<rgba 1 1 1 0.3>
     Index=$Particles.PartitionParticles.Partition
     Num=::NewEnumerator3
    )
    Lifetime=2
    Mesh Scale=&Multiply_VectorByFloat(
     Vector=&MergeHelper_VectorSwitch(
      0=<1.2 1.2 1.2>
      1=<1.225 1.225 1.225>
      2=<0.3 0.3 0.3>
      3=<15 15 15>
      4=<3.5 3.5 3.5>
      Index=$Particles.PartitionParticles.Partition
      Num=::NewEnumerator3
     )
     Float=$User.ScaleM
    )
    Mesh Scale Mode=::NewEnumerator4
    Position Offset:NiagaraPosition=$User.Offset
    Sprite Size Mode=::NewEnumerator3
    Uniform Sprite Size=500
    UsePositionOffset=true
    Write Material Random=true
   ) {module=InitializeParticle005}
   /SetVariables(Particles.DynMatParamDark:Vector4f=<1 4 100 0>) {isSetParameters=true}
   /SetVariables(Particles.VisibilityTag=$Particles.PartitionParticles.Partition) {isSetParameters=true}
   &InitialMeshOrientation(
    Mesh Orientation Mode=::NewEnumerator2
    Rotation=&MergeHelper_VectorSwitch(0=<0.01 0 0>, 1=<0.01 0 0>, 2=<0.35 0.5935 0>, 3=<0 0 0>, 4=<0.6 0 0>, Index=$Particles.PartitionParticles.Partition, Num=::NewEnumerator3)
    Use Rotation Vector=true
   )
  ]
  particleUpdateScript [
   &ParticleState(Kill Particles When Lifetime Has Elapsed=false, Let Infinitely Lived Particles Die When Emitter Deactivates=true)
   &ScaleMeshSize(Scale Factor=&VectorFromFloat(Value=&FloatFromCurve(FloatCurve=DataInterface() {propertyValues {Curve {keys [<at 0 0>,<at 0.1719 1.5>,<at 0.3065 1>]}}}, CurveIndex=$Emitter.Age))) {module=ScaleMeshSize005}
   &ScaleColor(
    Scale Alpha=&MergeHelper_FloatSwitch(
     0=&Multiply_Float(A=&RandomRangeFloat(Minimum=0.5, Maximum=2, Evaluation Type=::NewEnumerator1), B=$User.EnableValue)
     1=$User.EnableValue
     2=&Multiply_Float(A=&RandomRangeFloat(Minimum=0.5, Maximum=1), B=$User.EnableValue)
     3=&Multiply_Float(A=0.35, B=$User.EnableValue)
     4=$User.EnableValue
     Index=$Particles.PartitionParticles.Partition
     Num=::NewEnumerator3
    )
    Scale RGB=&MergeHelper_VectorSwitch(
     0=&VectorFromFloat(Value=8)
     1=<1 1 1>
     2=<1 1 1>
     3=&VectorFromFloat(Value=15)
     4=&VectorFromFloat(Value=1.5)
     Index=$Particles.PartitionParticles.Partition
     Num=::NewEnumerator3
    )
    ScaleA=true
    ScaleRGB=true
   ) {module=ScaleColor005}
   &ScaleMeshSize(Scale Factor=&MergeHelper_VectorSwitch(
     0=<1 1 1>
     1=&VectorFromFloat(Value=&FloatFromCurve(FloatCurve=DataInterface() {propertyValues {Curve {keys [<at 0 0>,<at 0.1719 1.5>,<at 0.3065 1>]}}}, CurveIndex=$Emitter.Age))
     2=&VectorFromFloat(Value=&Multiply_Float(
       A=&Waveform("Waveform [1]"=::NewEnumerator10, "[1] Phase Scale"=10, "[1] Frequency"=1, "[1] Amplitude Min/Max"=<0.9 1>)
       B=&FloatFromCurve(FloatCurve=DataInterface() {propertyValues {Curve {keys [<at 0 0>,<at 0.1719 1.5>,<at 0.3065 1>]}}}, CurveIndex=$Emitter.Age)
      ))
     3=&VectorFromFloat(Value=&Multiply_Float(
       A=&FloatFromCurve(FloatCurve=DataInterface() {propertyValues {Curve {keys [<at 0 1>,<at 15000 2.5>]}}}, CurveIndex=$Engine.Owner.LODDistance)
       B=&Multiply_Float(
        A=&Waveform("Waveform [1]"=::NewEnumerator10, "[1] Phase Scale"=10, "[1] Frequency"=1, "[1] Amplitude Min/Max"=<0.8 1>)
        B=&FloatFromCurve(FloatCurve=DataInterface() {propertyValues {Curve {keys [<at 0 0>,<at 0.1719 1.5>,<at 0.3065 1>]}}}, CurveIndex=$Emitter.Age)
       )
      ))
     4=&Multiply_VectorByFloat(
      Vector=&VectorFromFloat(Value=&Waveform("Waveform [1]"=::NewEnumerator10, "[1] Phase Scale"=10, "[1] Frequency"=1, "[1] Amplitude Min/Max"=<0.9 1>))
      Float=&FloatFromCurve(FloatCurve=DataInterface() {propertyValues {Curve {keys [<at 0 0>,<at 0.1719 1.5>,<at 0.3065 1>]}}}, CurveIndex=$Emitter.Age)
     )
     Index=$Particles.PartitionParticles.Partition
     Num=::NewEnumerator3
    )) {module=ScaleMeshSize009}
   &CameraOffset(Camera Offset Amount=&MakeCustomFloatFromBool(
     Bool=&SetBoolByIntComparison(A=$Particles.PartitionParticles.Partition, B=3, Comparison Type=::NewEnumerator4)
     True Float=-20
     False Float=&MakeCustomFloatFromBool(
      Bool=&SetBoolByIntComparison(A=$Particles.PartitionParticles.Partition, B=4, Comparison Type=::NewEnumerator4)
      True Float=200
      False Float=0
     )
    )) {module=CameraOffset001}
   /SetVariables(Particles.ColorDark:LinearColor=&MultiplyLinearColor(Linear Color=<rgba 0 0.0641 0.625 2.5>, Scale Factor=&MakeVector4(X=1, Y=1, Z=1, W=$User.EnableValue))) {isSetParameters=true}
   &DynamicMaterialParameters(
    Module.Index0Param0Write=true
    Module.Index0Param1Write=true
    Module.Index0Param2Write=true
    Module.Index0Param3Write=true
    Module.Index1Param0Write=true
    Module.Index1Param1Write=true
    Module.Index1Param2Write=true
    Module.Index1Param3Write=true
    Module.Index2Param0Write=true
    Module.Index2Param1Write=true
    Module.Index2Param2Write=true
    Module.Index2Param3Write=true
    Module.Index3Param0Write=true
    Module.Index3Param1Write=true
    Module.Index3Param2Write=true
    Module.Index3Param3Write=true
    Write Parameter Index 1=true
    Write Parameter Index 2=true
    Write Parameter Index 3=true
   ) {module=DynamicMaterialParameters001}
   /SetVariables(Particles.MergeDP1:Vector4f=&MergeHelper_Vector4Switch(
     0:Vector4f=<1 1 1 1>
     1:Vector4f=&MakeVector4(
      X=1
      Y=&Multiply_Float(A=0.2, B=&Divide_Float(A=$User.ScaleM, B=1.3))
      Z=100
      W=0
     )
     2:Vector4f=<1 1 1 1>
     3:Vector4f=<0 0 0 1>
     4:Vector4f=&MakeVector4(X=1, Y=&Multiply_Float(A=0.5, B=&Divide_Float(A=$User.ScaleM, B=1.3)), Z=1, W=0)
     5:Vector4f=?
     6:Vector4f=?
     7:Vector4f=?
     8:Vector4f=?
     9:Vector4f=?
     Index:NiagaraInt32=$Particles.PartitionParticles.Partition
     Num:MergeHelper_SwitcherEnum=::NewEnumerator3
    )) {isSetParameters=true}
  ]
  rendererClasses [&NiagaraMeshRendererProperties]
  renderers [
   {rendererClass=&NiagaraMeshRendererProperties, rendererIndex=0}
   {rendererClass=&NiagaraMeshRendererProperties, rendererIndex=1}
   {rendererClass=&NiagaraMeshRendererProperties, rendererIndex=2}
   {rendererClass=&NiagaraMeshRendererProperties, rendererIndex=3}
   {rendererClass=&NiagaraMeshRendererProperties, rendererIndex=4}
   {rendererClass=&NiagaraMeshRendererProperties, rendererIndex=5}
  ]
 }
 Shimmers {
  emitterUpdateScript [
   &EmitterState(Life Cycle Mode=::NewEnumerator1, Loop Duration=1, UseLoopDelay=true)
   &SpawnRate(SpawnRate=&Multiply_Float(A=4, B=$User.EnableValue), Use Spawn Probability=false)
  ]
  particleSpawnScript [
   /SetVariables(Particles.SpriteRotationRate=&RandomRangeFloat(Minimum=-0.03, Maximum=0.03, Evaluation Type=::NewEnumerator1)) {isSetParameters=true}
   /Niagara/Modules/Spawn/Initialization/InitializeParticle(
    Color=<rgba 3 3 3 1>
    Lifetime=&RandomRangeFloat(Minimum=3.7, Maximum=4.3, Evaluation Type=::NewEnumerator1)
    Module.Write Color=true
    Module.Write Lifetime=true
    Module.Write Mass=true
    Module.Write Material Random=true
    Module.Write Position=false
    Module.Write Scale=false
    Module.Write SpriteRotation=true
    Module.Write SpriteSize=true
    Sprite Rotation=&RandomRangeFloat(Minimum=-360, Maximum=360, Evaluation Type=::NewEnumerator1)
    Sprite Size=&RandomRangeVector2D(Minimum=<80 80>, Maximum=<120 120>, Evaluation Type=::NewEnumerator1)
   )
   &Color(Color=&MakeLinearColorFromVectorAndFloat(
     "Vector (RGB)"=&MakeVectorFromLinearColorRGB(Color=<rgba 0.05 0.35 0.5 1>)
     "Float (Alpha)"=&ClampFloat(Float=&RandomRangeFloat(Minimum=0.2, Maximum=0.3, Evaluation Type=::NewEnumerator1))
    ))
   /Niagara/Modules/Spawn/Location/SphereLocation(
    MaskSpawnGroup=false
    Module.Override Randomness=false
    Non Uniform Scale=&VectorFromFloat(Value=&FloatFromCurve(FloatCurve=DataInterface() {propertyValues {Curve {keys [<at 0 1>,<at 10000 1.4>]}}}, CurveIndex=$Engine.Owner.LODDistance))
    Sphere Radius=&RandomRangeFloat(Minimum=120, Maximum=160, Evaluation Type=::NewEnumerator1)
    Surface Only=true
   )
   &AddVelocity(Velocity=&Multiply_VectorByFloat(
     Vector=$Output.SphereLocation.SphereVector
     Float=&RandomRangeFloat(Minimum=-0.2, Maximum=-0.1, Evaluation Type=::NewEnumerator1)
    ))
   &DynamicMaterialParameters(
    Index 0 Param 1=&RandomRangeFloat(Minimum=0, Maximum=1, Evaluation Type=::NewEnumerator1)
    Index 0 Param 2=&RandomRangeFloat(Minimum=0, Maximum=1, Evaluation Type=::NewEnumerator1)
    Index 0 Param 3=1
    Module.Param0WriteEnabled=true
    Module.Param1WriteEnabled=true
    Module.Param2WriteEnabled=true
    Module.Param3WriteEnabled=true
   )
   &CameraOffset(Camera Offset Amount=&RandomRangeFloat(Minimum=50, Maximum=100, Evaluation Type=::NewEnumerator1))
  ]
  particleUpdateScript [
   &ParticleState()
   &SolveForcesAndVelocity(Clamp Velocity=false, Limit Acceleration=false)
   &DynamicMaterialParameters(
    Index 0 Param 3=1
    Index 0 Param 4=1
    Module.Param0WriteEnabled=true
    Module.Param1WriteEnabled=true
    Module.Param2WriteEnabled=true
    Module.Param3WriteEnabled=true
   ) {module=DynamicMaterialParameters001}
   &SpriteRotationRate(Rotation Rate=$Particles.SpriteRotationRate)
   &ScaleColor(
    Scale Alpha=&FloatFromCurve(
     FloatCurve=DataInterface() {propertyValues {Curve {keys [<at 0 0 RCIM_Cubic>,<at 0.5 1 RCIM_Cubic>,<at 1 0>]}}}
     Scale Curve=$User.EnableValue
    )
    ScaleA=true
    ScaleRGB=true
   )
  ]
  rendererClasses [&NiagaraSpriteRendererProperties]
  renderers [{rendererClass=&NiagaraSpriteRendererProperties}]
 }
 Spark {
  simTarget=GPUComputeSim
  emitterUpdateScript [
   &EmitterState(
    Inactive Response=::NewEnumerator1
    Life Cycle Mode=::NewEnumerator1
    Loop Behavior=::NewEnumerator1
    Loop Delay=&Add_Float(A=$User.Delay, B=0.1)
    Loop Duration=1
    UseLoopDelay=true
   )
   &SpawnBurst_Instantaneous(Spawn Count=&UniformRangedInt(Minimum=15, Maximum=20), Use Spawn Probability=false)
  ]
  particleSpawnScript [
   /Niagara/Modules/Spawn/Initialization/V2/InitializeParticle(
    Color=<rgba 0.3573 1.6587 5 5>
    Lifetime Max=0.4
    Lifetime Min=0.3
    Lifetime Mode=::NewEnumerator1
    Position Mode=::NewEnumerator1
    Sprite Size Max=<5.4951 14.4973>
    Sprite Size Min=<3.2988 6.5066>
    Sprite Size Mode=::NewEnumerator2
    UsePositionOffset=false
    Write Material Random=true
   ) {module=InitializeParticle001}
   /Niagara/Modules/Spawn/Location/V2/SphereLocation(
    MaskSpawnGroup=false
    Offset:NiagaraPosition=$User.Offset
    Override Seed=false
    Sphere Radius=20
    Surface Expansion Mode=::NewEnumerator1
    Surface Only=true
    Surface Only Band Thickness=1
    UseNonUniformScale=false
    UseOffset=true
    UseSphereOrientation=false
   ) {module=SphereLocation001}
   &AddVelocityFromPoint(
    Module.Use Falloff Distance=false
    Origin Offset:NiagaraPosition=$User.Offset
    Velocity Strength=&Multiply_Float(
     A=&RandomRangeFloat(Minimum=0.5, Maximum=1)
     B=&Lerp_Float(A=1001.3053, B=1928.6935, Alpha=1)
    )
   )
  ]
  particleUpdateScript [
   &ParticleState()
   &Drag(Drag=&UniformRangedFloat(Minimum=3, Maximum=5), Use Linear Drag=true, Use Rotational Drag=false)
   &SolveForcesAndVelocity(Clamp Velocity=false, Limit Acceleration=false)
   &ScaleSpriteSize(Scale Factor=&Vector2DFromCurve(
     Vector2Curve=DataInterface() {propertyValues {
       XCurve {keys [<at 0 1 RCIM_Cubic>, {arriveTangent=-0.5084, interpMode=RCIM_Cubic, leaveTangent=-0.5084, tangentMode=RCTM_User, time=1, value=0}]}
       YCurve {keys [{arriveTangent=-3.0155, interpMode=RCIM_Cubic, leaveTangent=-3.0155, tangentMode=RCTM_User, time=0, value=1}, <at 1 0 RCIM_Cubic>]}
      }}
     ScaleCurve=<1 1>
    ))
   &ScaleSpriteSizeBySpeed(Max Scale Factor=<0.5 5>, Min Scale Factor=<1 4>)
   &ScaleColor(
    Scale Alpha=&FloatFromCurve(FloatCurve=DataInterface() {propertyValues {Curve {keys [
         {tangentMode=RCTM_User, time=0, value=1}
         {arriveTangent=-0.0405, interpMode=RCIM_Cubic, leaveTangent=-0.0405, tangentMode=RCTM_User, time=0.7009, value=1}
         {arriveTangent=-8.0089, interpMode=RCIM_Cubic, leaveTangent=-8.0089, tangentMode=RCTM_User, time=1, value=0}
        ]}}})
    ScaleA=true
    ScaleRGB=true
   )
  ]
  rendererClasses [&NiagaraSpriteRendererProperties]
  renderers [{rendererClass=&NiagaraSpriteRendererProperties}]
 }
 TinyFrags {
  emitterUpdateScript [
   &EmitterState(
    Life Cycle Mode=::NewEnumerator1
    Loop Duration=&RandomRangeFloat(Minimum=1, Maximum=2, Evaluation Type=::NewEnumerator1)
    Recalculate Duration Each Loop=true
    UseLoopDelay=true
   )
   &SpawnRate(Spawn Probability=$User.EnableValue, SpawnRate=2, Use Spawn Probability=true)
   &SpawnBurst_Instantaneous(
    Spawn Count=&RandomRangeInt(Minimum=1, Maximum=2)
    Spawn Probability=$User.EnableValue
    Use Spawn Probability=true
   )
   &SpawnBurst_Instantaneous(
    Spawn Count=&RandomRangeInt(Minimum=0, Maximum=2)
    Spawn Probability=$User.EnableValue
    Spawn Time=0.25
    Use Spawn Probability=true
   ) {module=SpawnBurst_Instantaneous001}
   &SpawnBurst_Instantaneous(
    Spawn Count=&RandomRangeInt(Minimum=0, Maximum=2)
    Spawn Probability=$User.EnableValue
    Spawn Time=1.4
    Use Spawn Probability=true
   ) {module=SpawnBurst_Instantaneous002}
  ]
  particleSpawnScript [
   /SetVariables(Particles.MeshOrientation=&MakeQuatFromAxisAngle(
     Axis=&RandomRangeVector(Minimum=<0 0 0>, Maximum=<1 1 1>, Evaluation Type=::NewEnumerator1)
     AngleInDegrees=90
    )) {isSetParameters=true}
   /SetVariables(Particles.SpawnRotationVector=&RandomRangeVector(Minimum=<-0.15 -0.15 -0.15>, Maximum=<0.15 0.15 0.15>, Evaluation Type=::NewEnumerator1)) {isSetParameters=true}
   /Niagara/Modules/Spawn/Initialization/InitializeParticle(
    Lifetime=&RandomRangeFloat(Minimum=2, Maximum=2.5, Evaluation Type=::NewEnumerator1)
    Mesh Scale=&RandomRangeVector(Minimum=<0.5 0.5 2>, Maximum=<2 2 4>, Evaluation Type=::NewEnumerator1)
    Module.Write Color=true
    Module.Write Lifetime=true
    Module.Write Mass=true
    Module.Write Material Random=true
    Module.Write Position=false
    Module.Write Scale=true
    Module.Write SpriteRotation=false
    Module.Write SpriteSize=false
   )
   &ShapeLocation(
    MaskSpawnGroup=false
    Offset=<0 0 100>
    Offset Mode=::NewEnumerator0
    Override Seed=false
    Sphere Radius=150
    Sphere Surface Distribution=0.75
    UseOwnerScale=true
   )
   &Color(Color=&MakeLinearColorFromVectorAndFloat("Vector (RGB)"=<2 10 10>, "Float (Alpha)"=&ClampFloat(Float=1)))
   &AddVelocity(Velocity=&RandomRangeVector(Minimum=<-8 -8 -8>, Maximum=<8 8 8>, Evaluation Type=::NewEnumerator1)) {enabled=false}
   &AddVelocity(Velocity=&Multiply_VectorByFloat(
     Vector=$Particles.ShapeLocation.ShapeVector
     Float=&RandomRangeFloat(Minimum=-0.5, Maximum=-0.2, Evaluation Type=::NewEnumerator1)
    )) {module=AddVelocity001}
  ]
  particleUpdateScript [
   &ParticleState()
   &Drag(Drag=0.2, Use Linear Drag=true, Use Rotational Drag=false)
   &SolveForcesAndVelocity(Clamp Velocity=false, Limit Acceleration=false)
   &ScaleColor(
    Scale Alpha=&FloatFromCurve(FloatCurve=DataInterface() {propertyValues {Curve {keys [
         <at 0 0 RCIM_Cubic>
         {arriveTangent=1.4286, interpMode=RCIM_Cubic, leaveTangent=1.4286, time=0.2, value=1}
         {arriveTangent=-1.25, interpMode=RCIM_Cubic, leaveTangent=-1.25, time=0.7, value=1}
         <at 1 0>
        ]}}})
    ScaleA=true
    ScaleRGB=true
   )
   &UpdateMeshOrientation(Rotation Vector=$Particles.SpawnRotationVector) {module=UpdateMeshOrientation001}
   &ScaleMeshSize(Scale Factor=&VectorFromFloat(Value=&FloatFromCurve(
      FloatCurve=DataInterface() {propertyValues {Curve {keys [<at 0 0 RCIM_Cubic>, <at 0.2 1>, {arriveTangent=-1.25, interpMode=RCIM_Cubic, leaveTangent=-1.25, time=0.8, value=1}, <at 1 0>]}}}
      Scale Curve=$User.EnableValue
     )))
  ]
  rendererClasses [&NiagaraMeshRendererProperties]
  renderers [{rendererClass=&NiagaraMeshRendererProperties, rendererIndex=0}]
 }
}
summary {
 systemName=NS_ScubaBondOak_Rift
 emitters [
  {emitterName=Shimmers, rendererClasses [&NiagaraSpriteRendererProperties]}
  {emitterName=TinyFrags, rendererClasses [&NiagaraMeshRendererProperties]}
  {
   emitterName=FlaresAndEnergyRingAndPL
   rendererClasses [
    &NiagaraSpriteRendererProperties
    &NiagaraMeshRendererProperties
    &NiagaraLightRendererProperties
   ]
  }
  {emitterName=EnvChecker, rendererClasses [&NiagaraSpriteRendererProperties]}
  {
   emitterName=LinkLightning
   rendererClasses [
    &NiagaraSpriteRendererProperties
    &NiagaraRibbonRendererProperties
    &NiagaraMeshRendererProperties
   ]
  }
  {emitterName=LinkSpark, rendererClasses [&NiagaraSpriteRendererProperties]}
  {emitterName=Spark, simTarget=GPUComputeSim, rendererClasses [&NiagaraSpriteRendererProperties]}
  {emitterName=EnergyWave, rendererClasses [&NiagaraMeshRendererProperties, &NiagaraSpriteRendererProperties]}
  {emitterName=M001, rendererClasses [&NiagaraMeshRendererProperties]}
 ]
 userVariables [
  User.bIsFrontEnd {type=NiagaraBool}
  User.bIsFrontEndPreview {type=NiagaraBool}
  User.Delay {type=NiagaraFloat, defaultValue=0, description=""}
  User.EnableValue {type=NiagaraFloat, defaultValue=1, description=""}
  User.Offset {type=NiagaraPosition, description="", defaultValue {}}
  User.Open {type=NiagaraFloat, defaultValue=0, description=""}
  User.RandomArray {
   type=NiagaraDataInterfaceArrayFloat
   defaultValue {
    struct=NiagaraExt_VariableValue_DataInterface
    value {
     dataInterface=&self:NiagaraDataInterfaceArrayFloat_0
     dataInterfaceClass=&NiagaraDataInterfaceArrayFloat
    }
   }
  }
  User.ScaleM {type=NiagaraFloat, defaultValue=1.1, description=""}
  User.Simple Counter {
   type=NiagaraDataInterfaceSimpleCounter
   description=""
   defaultValue {
    struct=NiagaraExt_VariableValue_DataInterface
    value {
     dataInterface=&self:NiagaraDataInterfaceSimpleCounter_0
     dataInterfaceClass=/Script/Niagara.NiagaraDataInterfaceSimpleCounter
    }
   }
  }
  User.Vector 4 Array {
   type=NiagaraDataInterfaceArrayFloat4
   description=""
   defaultValue {
    struct=NiagaraExt_VariableValue_DataInterface
    value {
     dataInterface=&self:NiagaraDataInterfaceArrayFloat4_0
     dataInterfaceClass=/Script/Niagara.NiagaraDataInterfaceArrayFloat4
    }
   }
  }
 ]
}
