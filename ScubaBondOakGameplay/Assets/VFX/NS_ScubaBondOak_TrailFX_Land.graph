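# Niagara system for the ScubaBondOak "Land" skill trail, a single ribbon emitter that follows a target skeletal mesh.
# The sampled mesh velocity drives distance-based spawning, whose probability drops to zero for the local player, alongside a burst spawn.
# A local Noise module displaces particles through a tiling Perlin vector field.
# Ribbon width and alpha follow spawn-age curves and fade out once emitter age passes User.TotalSkillTime.
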NiagaraSystem /ScubaBondOakGameplay/Assets/VFX/NS_ScubaBondOak_TrailFX_Land
dependencies {
 usedDataInterfaces [/Script/Niagara.NiagaraDataInterfaceVectorField, &NiagaraDataInterfaceCurve, &NiagaraDataInterfaceSkeletalMesh]
 usedDynamicInputs []
 usedModules [
  &self:SystemSpawnScript.SystemScriptSource.SystemScriptGraph.NiagaraNodeAssignment_0.SetVariables_#1
  &SystemState
  /Niagara/Noise/Waveform
 ]
}
emitters {BrightTrail_JustSpawn_1Second {
  emitterSpawnScript [/SetVariables(Emitter.LastFramePosition:NiagaraPosition=$Engine.Emitter.SimulationPosition) {isSetParameters=true}]
  emitterUpdateScript [
   &EmitterState(Life Cycle Mode=::NewEnumerator1, Loop Behavior=::NewEnumerator1, UseLoopDelay=false)
   &SpawnBurst_Instantaneous(Use Spawn Probability=false, UseLoopCountLimit=false)
   &SampleSkeletalMesh(
    AdditionalYawPitchRoll=false
    Mesh Sampling Type:ENiagaraSkelSamplingModeFull=::NewEnumerator5
    Skeletal Mesh:NiagaraDataInterfaceSkeletalMesh=$System.Initial.TargetWSM
    Socket Sampling Mode:ENiagaraSocketSamplingMode=::NewEnumerator5
   )
   &SpawnPerUnit(
    Module.Use Max Movement Threshold=false
    Module.Use Movement Tolerance=false
    Spawn Probability:NiagaraFloat=&MakeCustomFloatFromBool(Bool=$User.IsLocalPlayer, True Float=0, False Float=1)
    Spawn Spacing=700
    Use Spawn Probability=true
    Velocity Vector:Vector3f=$Emitter.SampleSkeletalMesh.SampledVelocity
   )
   /SetVariables(Emitter.Axis:Vector3f=&NormalizeVector(
     Vector To Normalize=&Subtract_Position(A=$Engine.Emitter.SimulationPosition, B=$Emitter.LastFramePosition)
     Safe Normalize=true
     Safe Fallback Threshold=0
     Safe Fallback Vector=<0 0 0.001>
    )) {isSetParameters=true}
   /SetVariables(Emitter.LastFramePosition:NiagaraPosition=$Engine.Emitter.SimulationPosition) {isSetParameters=true}
   /SetVariables(Emitter.float:NiagaraFloat=&RandomRangeFloat(Minimum=0.2, Maximum=1, Evaluation Type=::NewEnumerator1)) {isSetParameters=true}
  ]
  particleSpawnScript [
   &InitializeParticle(
    Color=<rgba 0.3 0.3 0.3 0.8>
    Lifetime=6.5
    Ribbon Width=300
    Ribbon Width Mode=::NewEnumerator1
    Sprite Size Mode=::NewEnumerator3
    Uniform Sprite Size=500
    UsePositionOffset=false
    Write Material Random=true
   )
   &InitialMeshOrientation(
    Mesh Axis To Orient=<0 0 1>
    Mesh Orientation Mode=::NewEnumerator2
    Orientation Coordinate Space=::World
    Orientation Vector=$Emitter.Axis
    Use Rotation Vector=false
   )
   &SkeletalMeshLocation(MaskSpawnGroup=false, Mesh Sampling Type=::NewEnumerator5, Skeletal Mesh=$System.TargetWSM, Socket Sampling Mode=::NewEnumerator5)
   /SetVariables(Particles.SpawnAge:NiagaraFloat=$Emitter.Age) {isSetParameters=true}
  ]
  particleUpdateScript [
   &ParticleState()
   &self:Noise(
    Freq:Vector3f=<0.25 0.25 0.25>
    ScaleVF=400
    ScaleZ=1
    Vector Field=DataInterface() {propertyValues {Field=/NiagaraStaging/VectorFields/TilingPerlinFractal16, bTileX=true, bTileY=true, bTileZ=true}}
   ) {module=Noise}
   &ScaleRibbonWidth(Ribbon Width Scale=&ClampFloat(Float=&Subtract_Float(
      A=&FloatFromCurve(
       FloatCurve=DataInterface() {propertyValues {Curve {keys [
           {interpMode=RCIM_Cubic, leaveTangent=0.482, tangentMode=RCTM_Break, time=0, value=0.05}
           {arriveTangent=3.9252, interpMode=RCIM_Cubic, tangentMode=RCTM_Break, time=0.7, value=3}
          ]}}}
       CurveIndex=$Particles.OffsetLegnth
      )
      B=&FloatFromCurve(
       FloatCurve=DataInterface() {propertyValues {Curve {keys [<at 0 -0.1>,<at 1 1>]}}}
       CurveIndex=&Subtract_Float(A=$Emitter.Age, B=$User.TotalSkillTime)
      )
     )))
   &ScaleRibbonWidth(Ribbon Width Scale=&FloatFromCurve(
     FloatCurve=DataInterface() {propertyValues {Curve {keys [
         {arriveTangent=-0.5458, interpMode=RCIM_Cubic, leaveTangent=-0.5458, tangentMode=RCTM_User, time=0.01, value=0}
         {interpMode=RCIM_Cubic, tangentMode=RCTM_User, time=0.1, value=1}
         {arriveTangent=0.1861, interpMode=RCIM_Cubic, leaveTangent=0.1861, tangentMode=RCTM_User, time=0.9, value=1}
         {arriveTangent=-0.0696, interpMode=RCIM_Cubic, tangentMode=RCTM_Break, time=0.99, value=0}
        ]}}}
     CurveIndex=&Divide_Float(A=$Particles.SpawnAge, B=5)
    )) {module=ScaleRibbonWidth001}
   &ScaleColor(
    Scale Alpha=&Multiply_Float(
     A=&FloatFromCurve(FloatCurve=DataInterface() {propertyValues {Curve {keys [
          {arriveTangent=10.0087, interpMode=RCIM_Cubic, leaveTangent=10.0087, tangentMode=RCTM_User, time=0, value=0}
          {arriveTangent=-0.2353, leaveTangent=-0.2353, time=0.3, value=1}
          {arriveTangent=0.0224, leaveTangent=0.0224, time=1, value=1}
         ]}}})
     B=&FloatFromCurve(
      FloatCurve=DataInterface() {propertyValues {Curve {keys [
          {arriveTangent=3.6341, interpMode=RCIM_Cubic, leaveTangent=3.6341, tangentMode=RCTM_User, time=0.01, value=0}
          {interpMode=RCIM_Cubic, tangentMode=RCTM_User, time=0.2, value=1}
          {arriveTangent=-0.1388, interpMode=RCIM_Cubic, leaveTangent=-0.1388, tangentMode=RCTM_User, time=0.8, value=1}
          {arriveTangent=-0.0696, interpMode=RCIM_Cubic, tangentMode=RCTM_Break, time=0.99, value=0}
         ]}}}
      CurveIndex=&Divide_Float(A=$Particles.SpawnAge, B=5)
      Scale Curve=&FloatFromCurve(
       FloatCurve=DataInterface() {propertyValues {Curve {keys [<at 0 1>,<at 0.5 0>]}}}
       CurveIndex=&Subtract_Float(A=$Emitter.Age, B=$User.TotalSkillTime)
      )
     )
    )
    ScaleA=true
    ScaleRGB=true
   )
  ]
  rendererClasses [&NiagaraRibbonRendererProperties]
  renderers [{rendererClass=&NiagaraRibbonRendererProperties}]
 }}
summary {
 systemName=NS_ScubaBondOak_TrailFX_Land
 emitters [{emitterName=BrightTrail_JustSpawn_1Second, rendererClasses [&NiagaraRibbonRendererProperties]}]
 userVariables [
  User.IsLocalPlayer {type=NiagaraBool}
  User.TotalSkillTime {type=NiagaraFloat, defaultValue=5, description=""}
 ]
}
