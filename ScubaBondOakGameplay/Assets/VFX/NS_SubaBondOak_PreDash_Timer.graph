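# Niagara pre-dash timer effect for ScubaBond Oak: a short, blue energy-charge mesh burst that follows a skeletal mesh.
# A single EnergyChargeMesh emitter spawns in one burst, then re-samples its location on System.SK_Mesh every update.
# Curves over the particle's brief life spike then fade brightness and ramp down a dynamic material parameter.
# User.IsLocalPlayer makes the mesh smaller and much more transparent for the local player than for others.
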NiagaraSystem /ScubaBondOakGameplay/Assets/VFX/NS_SubaBondOak_PreDash_Timer
dependencies {
 usedDataInterfaces [&NiagaraDataInterfaceCurve, &NiagaraDataInterfaceSkeletalMesh]
 usedDynamicInputs [&Lerp_Float]
 usedModules [
  &self:SystemSpawnScript.SystemScriptSource.SystemScriptGraph.NiagaraNodeAssignment_0.SetVariables_#1
  &self:SystemSpawnScript.SystemScriptSource.SystemScriptGraph.NiagaraNodeAssignment_2.SetVariables_#2
  &SystemState
  &self:SystemSpawnScript.SystemScriptSource.SystemScriptGraph.NiagaraNodeAssignment_1.SetVariables_#3
 ]
}
emitters {EnergyChargeMesh {
  emitterUpdateScript [
   &EmitterState()
   &SpawnBurst_Instantaneous(Use Spawn Probability=false, UseLoopCountLimit=false)
  ]
  particleSpawnScript [
   &InitializeParticle(
    Color=<rgba 0.1066 0.5533 1 1>
    Lifetime=&MakeCustomFloatFromBool(Bool=$User.IsLocalPlayer, True Float=0.4, False Float=0.4)
    Mesh Scale=<0.1 1 1>
    Mesh Scale Mode=::NewEnumerator4
    UsePositionOffset=false
    Write Material Random=true
   )
   /SetVariables(Particles.NewFacing:Vector3f=<1 0 0>) {isSetParameters=true}
   &CameraOffset(Camera Offset Amount=100) {module=CameraOffset001}
   &InitialMeshOrientation(Mesh Axis To Orient=<-1 0 0>, Mesh Orientation Mode=::NewEnumerator2, Use Rotation Vector=false)
  ]
  particleUpdateScript [
   &ParticleState()
   &ScaleMeshSize(Scale Factor=&VectorFromFloat(Value=&MakeCustomFloatFromBool(Bool=$User.IsLocalPlayer, True Float=1, False Float=1.5)))
   &SkeletalMeshLocation(
    MaskSpawnGroup=false
    Mesh Sampling Type=::NewEnumerator5
    Orientation Sampling=::NewEnumerator1
    Skeletal Mesh=$System.SK_Mesh
    Socket Sampling Mode=::NewEnumerator5
   )
   &ScaleColor(
    Scale Alpha=&MakeCustomFloatFromBool(Bool=$User.IsLocalPlayer, True Float=0.2, False Float=1)
    Scale RGB=&VectorFromFloat(Value=&FloatFromCurve(
      FloatCurve=DataInterface() {propertyValues {Curve {keys [
          {arriveTangent=9.0725, interpMode=RCIM_Cubic, leaveTangent=9.0725, tangentMode=RCTM_User, time=0, value=0}
          {arriveTangent=-2.5249, interpMode=RCIM_Cubic, leaveTangent=-2.5249, tangentMode=RCTM_User, time=0.1408, value=0.9847}
          {arriveTangent=-0.0405, interpMode=RCIM_Cubic, leaveTangent=-0.0405, tangentMode=RCTM_User, time=1, value=0}
         ]}}}
      Scale Curve=30
     ))
    ScaleA=true
    ScaleRGB=true
   )
   &DynamicMaterialParameters(
    Index 0 Param 1=&FloatFromCurve(
     FloatCurve=DataInterface() {propertyValues {Curve {keys [
         {arriveTangent=-2.2086, interpMode=RCIM_Cubic, leaveTangent=-2.2086, tangentMode=RCTM_User, time=0, value=1}
         {interpMode=RCIM_Cubic, tangentMode=RCTM_User, time=1, value=0}
        ]}}}
     Scale Curve=0.4
    )
    Module.Index0Param0Write=true
    Module.Index0Param1Write=true
    Module.Index0Param2Write=true
    Module.Index0Param3Write=true
   )
  ]
  rendererClasses [&NiagaraMeshRendererProperties]
  renderers [{rendererClass=&NiagaraMeshRendererProperties, rendererIndex=0}]
 }}
summary {
 systemName=NS_SubaBondOak_PreDash_Timer
 emitters [{emitterName=EnergyChargeMesh, rendererClasses [&NiagaraMeshRendererProperties]}]
 userVariables [User.IsLocalPlayer {type=NiagaraBool, defaultValue=true}]
}
